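; Art tool Blueprint that batch-captures high-resolution screenshots of each hero in its Heroes array on a mannequin, apparently for character icons.
; BeginPlay stores the first FortPlayerMannequin found as the active mannequin.
; CaptureImages swaps in a gendered skeletal mesh, applies the hero type, waits, runs highresshot, then calls itself with the next index.
; Nothing here starts CaptureImages, HeroPortraits is empty, and completion turns off the HighResScreenshotMask show flag.
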(event Custom|HeroPortraits)

(event EventBeginPlay
  (bind _outactors (Actor|GetAllActorsOfClass "/Script/FortniteGame.FortPlayerMannequin"))
  (Variables|Default|SetMannequinActive (Utilities|Array|Get(acopy) _outactors)))

(event Custom|CaptureImages
  (bind _counter (Variables|Default|Getcounter))
  (bind _heroes (Variables|Default|GetHeroes))
  (bind _output (Utilities|Array|Get(acopy) _heroes _counter))
  (bind _skeletalmeshcomponent (Variables|/Script/Engine.SkeletalMeshActor|GetSkeletalMeshComponent (Variables|Default|GetMannequinActive)))
  (bind _mannequinactive (Variables|Default|GetMannequinActive))
  (switch Utilities|FlowControl|Switch|SwitchonEFortCustomGender (Variables|/Script/FortniteGame.FortWorkerType|GetGender _output)
    (:Male
      (Components|SkinnedMesh|SetSkinnedAssetandUpdate _skeletalmeshcomponent (Variables|Default|GetMaleSkeleton) true))
    (:Female
      (Components|SkinnedMesh|SetSkinnedAssetandUpdate _skeletalmeshcomponent (Variables|Default|GetFemaleSkeleton) true)))
  (Class|FortHeroExhibitActor|SetHeroType _mannequinactive _output)
  (Utilities|FlowControl|Delay 3)
  (Development|ExecuteConsoleCommand "highresshot 1")
  (Utilities|FlowControl|Delay 1)
  (Variables|Default|Setcounter (+ (Variables|Default|Getcounter) 1))
  (if (<= (Variables|Default|Getcounter) (Utilities|Array|LastIndex (Variables|Default|GetHeroes)))
    (CallFunction|CaptureImages)
    (else
      (Development|PrintString "================================ DONE")
      (Development|ExecuteConsoleCommand "ShowFlag.HighResScreenshotMask 0"))))

(fn ConstructionScript)
