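; GetLandscapeSplines is an art tool that copies landscape spline meshes into standalone actors, toggles their visibility, and deletes converted spline actors.
; GetLandscapeSplinesAsSingleActors spawns one BP_LandscapeSplineMeshActorCopy per segment, copying shape, mesh, materials and collision.
; ConvertLandscapeSplines chains segments from each proxy's first one by endpoint proximity, tracking reversed segments.
; It rebuilds that single chain per proxy as a SplineMeshActor_Landscape spline with copied tangents.
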(event Collision|EventActorBeginOverlap
  (params (OtherActor :type Actor%20Object%20Reference)))

(event EventTick
  (params (DeltaSeconds :type Float%20%28single-precision%29)))

(event Custom|GetLandscapeSplinesAsSingleActors
  (bind _outactors (Actor|GetAllActorsOfClass "/Script/Landscape.LandscapeProxy"))
  (for _array_element _outactors
    (for (_array_element_2 _array_index) (Actor|GetComponentsByClass _array_element "/Script/Engine.SplineMeshComponent")
      (Development|PrintString (Utilities|GetDisplayName _array_element_2))
      (bind _returnvalue (Game|SpawnActorfromClass "/Game/ArtTools/BP_LandscapeSplineMeshActorCopy.BP_LandscapeSplineMeshActorCopy_C" (Math|Transform|MakeTransform "0, 0, 0") "Undefined" "MultiplyWithRoot" 0 0))
      (SplineMesh|SetForwardAxis (Variables|/Script/Engine.SplineMeshActor|GetSplineMeshComponent _returnvalue) (SplineMesh|GetForwardAxis _array_element_2) true)
      (SplineMesh|SetStartandEnd (Variables|/Script/Engine.SplineMeshActor|GetSplineMeshComponent _returnvalue) (SplineMesh|GetStartPosition _array_element_2) (SplineMesh|GetStartTangent _array_element_2) (SplineMesh|GetEndPosition _array_element_2) (SplineMesh|GetEndTangent _array_element_2) true :!class /Script/Engine.SplineMeshComponent)
      (SplineMesh|SetStartOffset (Variables|/Script/Engine.SplineMeshActor|GetSplineMeshComponent _returnvalue) (SplineMesh|GetStartOffset _array_element_2) true)
      (SplineMesh|SetEndOffset (Variables|/Script/Engine.SplineMeshActor|GetSplineMeshComponent _returnvalue) (SplineMesh|GetEndOffset _array_element_2) true)
      (SplineMesh|SetStartScale (Variables|/Script/Engine.SplineMeshActor|GetSplineMeshComponent _returnvalue) (SplineMesh|GetStartScale _array_element_2) true :!class /Script/Engine.SplineMeshComponent)
      (SplineMesh|SetEndScale (Variables|/Script/Engine.SplineMeshActor|GetSplineMeshComponent _returnvalue) (SplineMesh|GetEndScale _array_element_2) true :!class /Script/Engine.SplineMeshComponent)
      (SplineMesh|SetStartRollRadians (Variables|/Script/Engine.SplineMeshActor|GetSplineMeshComponent _returnvalue) (SplineMesh|GetStartRoll _array_element_2) true)
      (SplineMesh|SetEndRollRadians (Variables|/Script/Engine.SplineMeshActor|GetSplineMeshComponent _returnvalue) (SplineMesh|GetEndRoll _array_element_2) true)
      (Components|StaticMesh|SetStaticMesh (Variables|/Script/Engine.SplineMeshActor|GetSplineMeshComponent _returnvalue) (Variables|/Script/Engine.StaticMeshComponent|GetStaticMesh _array_element_2))
      (SplineMesh|SetSplineUpDir (Variables|/Script/Engine.SplineMeshActor|GetSplineMeshComponent _returnvalue) (SplineMesh|GetSplineUpDir _array_element_2) true)
      (Transformation|SetActorLocation _returnvalue (Transformation|GetWorldLocation _array_element_2) false false)
      (Collision|SetCollisionProfileName (Variables|/Script/Engine.SplineMeshActor|GetSplineMeshComponent _returnvalue) (Collision|GetCollisionProfileName _array_element_2) true :!class /Script/Engine.PrimitiveComponent)
      (for (_array_element_3 _array_index_2) (Rendering|Material|GetMaterials _array_element_2)
        (Rendering|Material|SetMaterial (Variables|/Script/Engine.SplineMeshActor|GetSplineMeshComponent _returnvalue) _array_index_2 _array_element_3)))))

(event Custom|SetLandscapeSplinesInvisible
  ;[1 1 1] Visibilty of LandscapeSplines
  (bind _outactors (Actor|GetAllActorsOfClass "/Script/Landscape.LandscapeProxy"))
  (for _array_element _outactors
    (for _array_element_2 (Actor|GetComponentsByClass _array_element "/Script/Engine.SplineMeshComponent")
      (Rendering|SetVisibility _array_element_2 false false)
      (Development|SetHiddeninGame _array_element_2 true false))))

(event Custom|SetLandscapeSplinesVisible
  ;[1 1 1] Visibilty of LandscapeSplines
  (bind _outactors (Actor|GetAllActorsOfClass "/Script/Landscape.LandscapeProxy"))
  (for _array_element _outactors
    (for _array_element_2 (Actor|GetComponentsByClass _array_element "/Script/Engine.SplineMeshComponent")
      (Rendering|SetVisibility _array_element_2 true false)
      (Development|SetHiddeninGame _array_element_2 false false))))

(event Custom|Delete%20all%20Splines%20
  ;[1 1 1] Delete all Splines Created for Debuging
  (bind _outactors (Actor|GetAllActorsOfClass "/Game/ArtTools/SplineMeshActor_Landscape.SplineMeshActor_Landscape_C"))
  (for _array_element _outactors
    (Actor|DestroyActor _array_element)))

(event Custom|ConvertLandscapeSplines
  (bind _tempsplinepoints (Variables|Default|GetTempSplinePoints))
  (bind _splinesconnected_ordered (Variables|Default|GetSplinesConnected_Ordered))
  (Utilities|Array|Clear (Variables|Default|GetSplinesConnected_Ordered))
  (bind _outactors (Actor|GetAllActorsOfClass "/Script/Landscape.LandscapeProxy"))
  (Utilities|Array|Clear (Variables|Default|GetSplinesConnected))
  (for _array_element _outactors
    (Variables|Default|SetSplinesConnected (Actor|GetComponentsByClass _array_element "/Script/Engine.SplineMeshComponent"))
    (if (> (Utilities|Array|Length (Variables|Default|GetSplinesConnected)) 0)
      (Utilities|Array|Clear _splinesconnected_ordered)
      (Utilities|Array|Add _splinesconnected_ordered (Utilities|Array|Get(acopy) (Variables|Default|GetSplinesConnected)))
      ;[1 1 1] 4 Checks forward and back ward to get all connections
      (CallFunction|CheckConnectingSplinesForward)
      (CallFunction|CheckConnectingSplinesBackward)
      (CallFunction|CheckConnectingSplinesForward)
      (CallFunction|CheckConnectingSplinesBackward)
      (bind _returnvalue (Game|SpawnActorfromClass "/Game/ArtTools/SplineMeshActor_Landscape.SplineMeshActor_Landscape_C" (Math|Transform|MakeTransform "0, 0, 0") "Undefined" "MultiplyWithRoot" 0 0))
      (Utilities|Array|Clear (Variables|Default|GetTempSplinePoints))
      (CallFunction|MakeSplinePointArray _tempsplinepoints (Variables|Default|GetSplinesConnected_Ordered))
      (Spline|SetSplinePoints (Variables|/Game/ArtTools/SplineMeshActor_Landscape.SplineMeshActor_Landscape_C|GetSpline _returnvalue) _tempsplinepoints "World" true)
      (CallFunction|SetTangentsonSplinePointList (Variables|Default|GetSplinesConnected_Ordered) _returnvalue)
      ;[1 1 1] Update the spline mesh
      (Class|SplineMeshActorLandscape|UpdateMeshfromSplineData _returnvalue)
      ;[1 1 1] Remove copied splines form array of Splines
      (for _array_element_2 (Variables|Default|GetSplinesConnected_Ordered)
        (Utilities|Array|RemoveIndex (Variables|Default|GetSplinesConnected) (Utilities|Array|FindItem (Variables|Default|GetSplinesConnected) _array_element_2))))))

(fn CheckConnectingSplinesBackward
  (bind _splinesconnected_ordered (Variables|Default|GetSplinesConnected_Ordered))
  (bind _output (Utilities|Array|Get(acopy) _splinesconnected_ordered))
  (bind _splinesconnected_mirrored (Variables|Default|GetSplinesConnected_Mirrored))
  (bind _returnvalue (Utilities|Array|ContainsItem _splinesconnected_mirrored _output))
  (bind _returnvalue_2 (SplineMesh|GetStartPosition _output))
  (bind _returnvalue_3 (SplineMesh|GetEndPosition _output))
  (bind _returnvalue_4 (Transformation|GetWorldLocation _output))
  (bind _returnvalue_5 (+ _returnvalue_4 _returnvalue_2))
  (bind _returnvalue_6 (+ _returnvalue_4 _returnvalue_3))
  (bind _returnvalue_7 (Math|Vector|SelectVector _returnvalue_6 _returnvalue_5 _returnvalue))
  (for _array_element (Variables|Default|GetSplinesConnected)
    (if (Utilities|Equal(Object) _array_element _output)
      _
      (elif (Math|Vector|Equal(Vector) (+ (Transformation|GetWorldLocation _array_element) (SplineMesh|GetEndPosition _array_element)) _returnvalue_7 5.0)
        (Utilities|Array|Insert (Variables|Default|GetSplinesConnected_Ordered) _array_element)
        (elif (Math|Vector|Equal(Vector) (+ (Transformation|GetWorldLocation _array_element) (SplineMesh|GetStartPosition _array_element)) _returnvalue_7 5.0)
          (Utilities|Array|Insert (Variables|Default|GetSplinesConnected_Ordered) _array_element)
          (Utilities|Array|Add (Variables|Default|GetSplinesConnected_Mirrored) _array_element))))))

(fn CheckConnectingSplinesForward
  (bind _splinesconnected_ordered (Variables|Default|GetSplinesConnected_Ordered))
  (bind _returnvalue (Utilities|Array|LastIndex _splinesconnected_ordered))
  (bind _output (Utilities|Array|Get(acopy) _splinesconnected_ordered _returnvalue))
  (bind _splinesconnected_mirrored (Variables|Default|GetSplinesConnected_Mirrored))
  (bind _returnvalue_2 (Utilities|Array|ContainsItem _splinesconnected_mirrored _output))
  (bind _returnvalue_3 (Transformation|GetWorldLocation _output))
  (bind _returnvalue_4 (SplineMesh|GetEndPosition _output))
  (bind _returnvalue_5 (SplineMesh|GetStartPosition _output))
  (bind _returnvalue_6 (+ _returnvalue_3 _returnvalue_4))
  (bind _returnvalue_7 (+ _returnvalue_3 _returnvalue_5))
  (bind _returnvalue_8 (Math|Vector|SelectVector _returnvalue_7 _returnvalue_6 _returnvalue_2))
  (for _array_element (Variables|Default|GetSplinesConnected)
    (if (Utilities|Equal(Object) _array_element _output)
      _
      (elif (Math|Vector|Equal(Vector) (+ (Transformation|GetWorldLocation _array_element) (SplineMesh|GetStartPosition _array_element)) _returnvalue_8 5.0)
        (Utilities|Array|Add (Variables|Default|GetSplinesConnected_Ordered) _array_element)
        (elif (Math|Vector|Equal(Vector) (+ (Transformation|GetWorldLocation _array_element) (SplineMesh|GetEndPosition _array_element)) _returnvalue_8 5.0)
          (Utilities|Array|Add (Variables|Default|GetSplinesConnected_Ordered) _array_element)
          (Utilities|Array|Add (Variables|Default|GetSplinesConnected_Mirrored) _array_element))))))

(fn MakeSplinePointArray
  (params (SplinePointArray :type Array%20of%20Vectors) (SplineCompenentArray :type Array%20of%20Spline%20Mesh%20Component%20Object%20References))
  (bind _splinesconnected_mirrored (Variables|Default|GetSplinesConnected_Mirrored))
  (for (_array_element _array_index) SplineCompenentArray
    (Utilities|Array|Add SplinePointArray (+ (Transformation|GetWorldLocation _array_element) (Math|Vector|SelectVector (SplineMesh|GetEndPosition _array_element) (SplineMesh|GetStartPosition _array_element) (Utilities|Array|ContainsItem _splinesconnected_mirrored _array_element))))
    (if (== _array_index (Utilities|Array|LastIndex SplineCompenentArray))
      (Utilities|Array|Add SplinePointArray (+ (Transformation|GetWorldLocation _array_element) (Math|Vector|SelectVector (SplineMesh|GetStartPosition _array_element) (SplineMesh|GetEndPosition _array_element) (Utilities|Array|ContainsItem _splinesconnected_mirrored _array_element)))))))

(fn SetSplineData
  (params (SplineMeshActor :type Spline%20Mesh%20Actor%20Landscape%20Object%20Reference))
  (for (_array_element _array_index) (Variables|Default|GetSplinesConnected_Ordered)
    _))

(fn SetTangentsOnSplinePointList
  (params (SplineComponentArray :type Array%20of%20Spline%20Mesh%20Component%20Object%20References) (SplineMeshActor :type Spline%20Mesh%20Actor%20Landscape%20Object%20Reference))
  (bind _spline (Variables|/Game/ArtTools/SplineMeshActor_Landscape.SplineMeshActor_Landscape_C|GetSpline SplineMeshActor))
  (for (_array_element _array_index) SplineComponentArray
    (Spline|SetTangentatSplinePoint _spline _array_index (Math|Vector|SelectVector (SplineMesh|GetEndTangent _array_element) (SplineMesh|GetStartTangent _array_element) (Utilities|Array|ContainsItem (Variables|Default|GetSplinesConnected_Mirrored) _array_element)) "Local" true)
    (if (== _array_index (Utilities|Array|LastIndex SplineComponentArray))
      (Spline|SetTangentatSplinePoint _spline (+ _array_index 1) (SplineMesh|GetEndTangent _array_element) "Local" true))))

(fn ConstructionScript)
