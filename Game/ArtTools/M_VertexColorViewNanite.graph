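# M_VertexColorViewNanite is a Default Lit, Nanite-enabled material in ArtTools, named as a viewer for mesh vertex colors.
# Its graph holds only a VertexColor node and a Constant at its default value, with no texture or parameter nodes.
# The file records no links from either node to the material outputs, so the channel each one drives is not shown.
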Material /Game/ArtTools/M_VertexColorViewNanite
graph {
 properties {bUsedWithNanite=true, pixelDepthOffsetMode=PDOM_Legacy, usedShadingModels=Default Lit}
 nodes [
  Constant()
  VertexColor()
 ]
}
