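; Function library for the Playset capture art tool, which frames actors and screenshots them, apparently for icons.
; It merges bounding spheres of IconCapture-tagged meshes, falling back to one mesh, then rescales the actor to a radius derived from FOV and zoom.
; The camera centers on those bounds, turns by axis tags (Y-, X-, X+) and backs off by an FOV-only distance.
; Helpers list maps by directory and save PNGs through highresshot.
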(fn CalculateCaptureItemDistance
  (params (Field%20Of%20View :type Float%20%28double-precision%29))
  (return (* (Math|Trig|Cos(Degrees) (/ Field%20Of%20View 2.0)) -1e+03)))

(fn CalculateCaptureItemRadius
  (params (Field%20Of%20View :type Float%20%28double-precision%29 :default 9e+01) (Zoom :type Float%20%28double-precision%29 :default 1.0))
  (return (* Zoom (* (Math|Trig|Sin(Degrees) (/ Field%20Of%20View 2.0)) 5e+02))))

(fn CombineBounds
  (params (CenterA :type Vector) (RadiusA :type Float%20%28double-precision%29) (CenterB :type Vector) (RadiusB :type Float%20%28double-precision%29))
  ;[1 1 1] Handle 0 radius bounds.
  (if (<= (Variables|GetRadiusA) 0.0)
    (return (Variables|GetCenterB) (Variables|GetRadiusB))
    (elif (<= (Variables|GetRadiusB) 0.0)
      (return (Variables|GetCenterA) (Variables|GetRadiusA))
      (else
        (Utilities|FlowControl|Sequence
          (:then_0
            ;[1 1 1] Compute intermediate values.
            (bind _output_get (LocalVariables|SetABn (Math|Vector|Normalize (- (Variables|GetCenterB) (Variables|GetCenterA)))))
            (LocalVariables|SetBAn (* _output_get -1.0))
            (LocalVariables|SetC (+ (Variables|GetCenterB) (* (LocalVariables|GetABn) (Variables|GetRadiusB))))
            (LocalVariables|SetD (+ (Variables|GetCenterA) (* (LocalVariables|GetBAn) (Variables|GetRadiusA))))
            (LocalVariables|SetE (+ (Variables|GetCenterA) (* (LocalVariables|GetABn) (Variables|GetRadiusA))))
            (LocalVariables|SetF (+ (Variables|GetCenterB) (* (LocalVariables|GetBAn) (Variables|GetRadiusB))))
            (LocalVariables|SetAC (- (LocalVariables|GetC) (Variables|GetCenterA)))
            (LocalVariables|SetAF (- (LocalVariables|GetF) (Variables|GetCenterA)))
            (LocalVariables|SetBD (- (LocalVariables|GetD) (Variables|GetCenterB)))
            (LocalVariables|SetBE (- (LocalVariables|GetE) (Variables|GetCenterB))))
          (:then_1
            ;[1 1 1] Compute endpoints.
            (if (> (Math|Vector|VectorLength (LocalVariables|GetAF)) (Math|Vector|VectorLength (LocalVariables|GetAC)))
              (LocalVariables|SetEndpointB (LocalVariables|GetF))
              (else
                (LocalVariables|SetEndpointB (LocalVariables|GetC)))))
          (:then_2
            (if (> (Math|Vector|VectorLength (LocalVariables|GetBE)) (Math|Vector|VectorLength (LocalVariables|GetBD)))
              (LocalVariables|SetEndpointA (LocalVariables|GetE))
              (else
                (LocalVariables|SetEndpointA (LocalVariables|GetD)))))
          (:then_3
            ;[1 1 1] Calculate new center & radius
            (return (+ (/ (- (LocalVariables|GetEndpointB) (LocalVariables|GetEndpointA)) 2.0) (LocalVariables|GetEndpointA)) (Math|Vector|VectorLength (/ (- (LocalVariables|GetEndpointB) (LocalVariables|GetEndpointA)) 2.0)))))))))

(fn GetCaptureComponents
  (params (Actor :type Actor%20Object%20Reference))
  (bind _returnvalue (Actor|GetComponentsbyTag Actor "/Script/Engine.MeshComponent" "IconCapture"))
  (if (> (Utilities|Array|Length _returnvalue) 0)
    (return _returnvalue)
    (else
      (return (Utilities|Array|MakeArray (Actor|GetComponentbyClass Actor "/Script/Engine.MeshComponent"))))))

(fn GetCaptureItemBounds
  (params (Actor :type Actor%20Object%20Reference))
  (bind _primitives (CallFunction|GetCaptureComponents Actor))
  (for _array_element _primitives
    (bind (_center _radius) (CallFunction|CombineBounds (LocalVariables|GetItemCenter) (LocalVariables|GetItemRadius) (Collision|GetComponentBounds _array_element) (.SphereRadius (Collision|GetComponentBounds _array_element))))
    (LocalVariables|SetItemCenter _center)
    (LocalVariables|SetItemRadius _radius))
  (return (LocalVariables|GetItemCenter) (LocalVariables|GetItemRadius)))

(fn GetCaptureItemCenter
  (params (Actor :type Actor%20Object%20Reference))
  (bind _center (CallFunction|GetCaptureItemBounds Actor))
  (return _center))

(fn GetCaptureMapList
  (params (Directory :type Name))
  (bind _outassetdata (AssetRegistry|GetAssetsbyPath (AssetRegistry|GetAssetRegistry) Directory true false))
  (for _array_element _outassetdata
    (if (Utilities|Name|Equal(Name) (Utilities|String|StringToName (Utilities|BreakTopLevelAssetPath (.AssetClassPath (Utilities|Struct|BreakAssetData _array_element)))) "/Script/Engine.World")
      (Utilities|Array|Add (LocalVariables|GetMapList) (Utilities|Struct|BreakAssetData _array_element))))
  (return (LocalVariables|GetMapList)))

(fn GetCaptureMapListfromArray
  (params (DirectoryArray :type Array%20of%20Names))
  (for _array_element DirectoryArray
    (bind _outassetdata (AssetRegistry|GetAssetsbyPath (AssetRegistry|GetAssetRegistry) _array_element true false))
    (for _array_element_2 _outassetdata
      (if (Utilities|Name|Equal(Name) (Utilities|String|StringToName (Utilities|BreakTopLevelAssetPath (.AssetClassPath (Utilities|Struct|BreakAssetData _array_element_2)))) "/Script/Engine.World")
        (Utilities|Array|Add (LocalVariables|GetMapList) (Utilities|Struct|BreakAssetData _array_element_2)))))
  (return (LocalVariables|GetMapList)))

(fn GetCaptureMapName
  (params (PackageName :type Name))
  (bind _returnvalue (Utilities|String|ToString(Name) PackageName))
  (bind _result (Math|Integer|IncrementInt (Utilities|String|FindSubstring _returnvalue "/" false true))
    (:%20%20
      (return (Utilities|String|RightChop _returnvalue _result)))))

(fn GetForwardRotation
  (params (Actor :type Actor%20Object%20Reference))
  (Utilities|FlowControl|Sequence
    (:then_0
      (if (Actor|ActorHasTag Actor "Y-")
        (LocalVariables|SetCameraYaw 9e+01)
        (elif (Actor|ActorHasTag Actor "X-")
          (LocalVariables|SetCameraYaw 0.0)
          (elif (Actor|ActorHasTag Actor "X+")
            (LocalVariables|SetCameraYaw 1.8e+02)))))
    (:then_1
      (return (Math|Rotator|MakeRotator 0.0 0.0 (LocalVariables|GetCameraYaw))))))

(fn PoseCaptureCamera
  (params (CaptureCamera :type Actor%20Object%20Reference) (CaptureItem :type Actor%20Object%20Reference) (Field%20Of%20View :type Float%20%28double-precision%29 :default 9e+01) (Rotation :type Rotator))
  (Transformation|SetActorTransform CaptureCamera (Transformation|GetActorTransform CaptureItem) false true)
  (bind _center (CallFunction|GetCaptureItemCenter CaptureItem))
  (Transformation|SetActorLocation CaptureCamera _center false true)
  (bind _rotation (CallFunction|GetForwardRotation CaptureItem))
  (Transformation|AddActorLocalRotation CaptureCamera (Math|Rotator|CombineRotators Rotation _rotation) false false)
  (bind _distance (CallFunction|CalculateCaptureItemDistance Field%20Of%20View))
  (Transformation|AddActorLocalOffset CaptureCamera (Math|Vector|MakeVector _distance) false false))

(fn PoseCaptureItem
  (params (Actor :type Actor%20Object%20Reference) (Field%20Of%20View :type Float%20%28double-precision%29) (Zoom :type Float%20%28double-precision%29))
  (bind _radius (CallFunction|CalculateCaptureItemRadius Field%20Of%20View Zoom))
  (CallFunction|ScaleCaptureItem Actor _radius))

(fn SaveScreenshot
  (params (Filename :type String) (Resolution :type Vector%202D%20Structure :default "(X=1024.000000,Y=1024.000000)"))
  (bind (_x _y) (Math|Vector2D|BreakVector2D Resolution))
  (Development|ExecuteConsoleCommand 0 "r.screenpercentage 100")
  (Development|ExecuteConsoleCommand 0 (Utilities|String|Append "highresshot filename=" Filename ".png " (Utilities|String|ToString(Integer) (Math|Float|Truncate _x)) "x" (Utilities|String|ToString(Integer) (Math|Float|Truncate _y)) " 0 0 0 0 1 0 0") (Game|GetPlayerController 0)))

(fn ScaleCaptureItem
  (params (Actor :type Actor%20Object%20Reference) (Radius :type Float%20%28double-precision%29))
  (bind (_center _sphere_radius) (CallFunction|GetCaptureItemBounds Actor))
  (Transformation|SetActorScale3D Actor (* (Transformation|GetActorScale3D Actor) (/ Radius _sphere_radius))))
