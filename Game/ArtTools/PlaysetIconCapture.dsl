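; Camera actor tool that batch-renders playset set and item icons to disk, then quits.
; It gathers maps from input folders or plugins, streaming each in and chaining through CaptureComplete.
; Per map, Setup runs console commands, gathers capture actors, enables custom depth and freezes time of day.
; It shoots from the scene camera, then recursively isolates one actor per class, framed using default or map-specific overrides.
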(event EventBeginPlay
  (EventDispatchers|BindEventtoCaptureComplete (delegate Custom|MapCaptureComplete))
  (CallFunction|GenerateCaptureMapList)
  (CallFunction|MakeMapIcons))

(event Custom|MapCaptureComplete
  ;[1 1 1] MapCaptureComplete
  (Variables|/Script/Engine.LevelStreaming|SetbShouldBeLoaded false (Variables|Transient|GetStreamedLevel))
  (bind _result (Math|Integer|IncrementInt (Variables|Transient|GetMapIndex))
    (:%20%20
      (Variables|Transient|SetMapIndex _result)
      (CallFunction|MakeMapIcons))))

(event Custom|MakeMapIcons
  (bind _mapindex (Variables|Transient|GetMapIndex))
  (bind _capturemaplist (Variables|Transient|GetCaptureMapList))
  (bind _output (Utilities|Array|Get(acopy) _capturemaplist _mapindex))
  ;[1 1 1] MakeMapIcons
  (if (< _mapindex (Utilities|Array|Length _capturemaplist))
    (bind (_boutsuccess _returnvalue) (LevelStreaming|LoadLevelInstance(byName) 0 (Utilities|String|ToString(Name) _output)))
    (bind _mapname (Class|PlaysetCaptureLibrary|GetCaptureMapName _output))
    (Variables|Transient|SetOutputSubdirectory _mapname)
    (Variables|Transient|SetPlaysetName _mapname)
    (Variables|Transient|SetStreamedLevel _returnvalue)
    (EventDispatchers|BindEventtoOnLevelShown _returnvalue (delegate Custom|OnLevelLoaded))
    (else
      (Game|QuitGame 0))))

(event Custom|OnLevelLoaded
  ;[1 1 1] MakeMapIcons
  (CallFunction|BeginCapture))

(event Custom|WarmScene
  ;[1 1 1] Aim the camera so things are in the frustum to prevent frustum culling, so texture streaming starts.
  (CallFunction|UseSceneCamera))

(event Custom|BeginCapture
  ;[1 1 1] Begin Capture
  (Class|BPFuncHQScreenshot|FortniteScreenShotCapture 0)
  (CallFunction|Setup)
  (CallFunction|WarmScene)
  (Utilities|FlowControl|Delay 0 0.5)
  (CallFunction|ScreenshotLevel)
  (if (Variables|IconCapture|GetGenerate%20Item%20Icons)
    (Utilities|FlowControl|Delay 0)
    (CallFunction|ScreenshotItems)
    (else
      (Utilities|FlowControl|Delay 0 1.0)
      (Default|CallCaptureComplete))))

(event Custom|Screenshot%20Level
  ;[1 1 1] Screenshot the Entire Playset
  (if (Variables|IconCapture|GetGenerate%20Set%20Icons)
    (CallFunction|UseSceneCamera)
    (CallFunction|SaveScreenshot (Variables|Transient|GetPlaysetName))))

(event Custom|Screenshot%20Items
  (bind _captureitems (Variables|Transient|GetCaptureItems))
  ;[1 1 1] Screenshot each Playset Item
  (CallFunction|HideAllBut 0 _captureitems)
  (bind _pruneditemlist (CallFunction|RemoveDuplicates _captureitems))
  (Variables|Transient|SetCaptureItems _pruneditemlist)
  (CallFunction|ScreenshotItem 0))

(event Custom|Screenshot%20Item
  (params (Index :type Integer))
  (bind _captureitems (Variables|Transient|GetCaptureItems))
  (bind _output (Utilities|Array|Get(acopy) _captureitems Index))
  ;[1 1 1] Recursively take a screenshot of each item in PlaysetItems
  (if (< Index (Utilities|Array|Length _captureitems))
    (Utilities|FlowControl|Delay 0 0.5)
    (CallFunction|FrameItem _output)
    (CallFunction|HideAllBut _output _captureitems)
    (CallFunction|ScreenshotSingleItem _output)
    (CallFunction|ScreenshotItem (+ Index 1))
    (else
      (Utilities|FlowControl|Delay 0 3.0)
      (Default|CallCaptureComplete))))

(fn EnableCustomDepth
  (params (Item :type Actor%20Object%20Reference) (Enabled :type Boolean))
  (for _array_element (Actor|GetComponentsByClass Item "/Script/Engine.PrimitiveComponent")
    (Rendering|SetRenderCustomDepth _array_element Enabled)))

(fn FindItemParams
  (params (Actor :type Actor%20Object%20Reference))
  (LocalVariables|SetOutRotation (Variables|IconCapture|GetDefault%20Rotation))
  (LocalVariables|SetOutFOV (Variables|IconCapture|GetDefault%20Field%20Of%20View))
  (LocalVariables|SetOutZoom (Variables|IconCapture|GetDefault%20Zoom))
  ;[1 1 1] Look first for a Blueprint default override.
  (for _array_element (Variables|Default|GetDefaultItemOverrides)
    (if (Utilities|ClassIsChildOf (Utilities|GetClass Actor) (Utilities|Struct|BreakPlaysetIconItemParams _array_element))
      (LocalVariables|SetOutRotation (.Rotation_10_795C7063413DECB9F02717B7A65AE159 (Utilities|Struct|BreakPlaysetIconItemParams _array_element)))
      (LocalVariables|SetOutFOV (.FieldOfView_9_B770FCDE46665DFF912E81B7A9AF2227 (Utilities|Struct|BreakPlaysetIconItemParams _array_element)))
      (LocalVariables|SetOutZoom (.Zoom_13_61C8A35D49ECC35DCD1DA083421624A4 (Utilities|Struct|BreakPlaysetIconItemParams _array_element)))))
  ;[1 1 1] Next look for a Map-specific override.
  (for _array_element_2 (Variables|IconCapture|GetItemOverrides)
    (if (Utilities|ClassIsChildOf (Utilities|GetClass Actor) (Utilities|Struct|BreakPlaysetIconItemParams _array_element_2))
      (LocalVariables|SetOutRotation (.Rotation_10_795C7063413DECB9F02717B7A65AE159 (Utilities|Struct|BreakPlaysetIconItemParams _array_element_2)))
      (LocalVariables|SetOutFOV (.FieldOfView_9_B770FCDE46665DFF912E81B7A9AF2227 (Utilities|Struct|BreakPlaysetIconItemParams _array_element_2)))
      (LocalVariables|SetOutZoom (.Zoom_13_61C8A35D49ECC35DCD1DA083421624A4 (Utilities|Struct|BreakPlaysetIconItemParams _array_element_2)))))
  (return (LocalVariables|GetOutFOV) (LocalVariables|GetOutZoom) (LocalVariables|GetOutRotation)))

(fn FrameItem
  (params (Item :type Actor%20Object%20Reference))
  (bind (_fov _zoom _rotation) (CallFunction|FindItemParams Item))
  (Class|PlaysetCaptureLibrary|PoseCaptureItem Item _fov _zoom)
  (Class|PlaysetCaptureLibrary|PoseCaptureCamera self Item _fov _rotation)
  (Camera|SetFieldOfView (Variables|CameraActor|GetCameraComponent) _fov))

(fn GenerateCaptureMapList
  (if (Variables|IconCapture|GetGenerateUserFolderList)
    (for (_array_element _array_index) (Variables|IconCapture|GetUserFolderList)
      (Utilities|Array|SetArrayElem (Variables|IconCapture|GetUserFolderList) _array_index (Utilities|String|StringToName (Utilities|String|Append "/Game/" (Utilities|String|ToString(Name) _array_element)))))
    (bind _maps (Class|PlaysetCaptureLibrary|GetCaptureMapListfromArray (Variables|IconCapture|GetUserFolderList)))
    (Variables|Transient|SetCaptureMapList _maps)
    (else
      (bind _maps_2 (Class|PlaysetCaptureLibrary|GetCaptureMapList (Utilities|String|StringToName (Utilities|String|Append "/Game/" (Variables|IconCapture|GetInput%20Directory)))))
      (Variables|Transient|SetCaptureMapList _maps_2)
      (for _array_element_2 (Variables|IconCapture|GetPlugins)
        (bind _maps_3 (Class|PlaysetCaptureLibrary|GetCaptureMapList (Utilities|String|StringToName _array_element_2)))
        (Utilities|Array|AppendArray (Variables|Transient|GetCaptureMapList) _maps_3)))))

(fn GetAllCaptureActors
  (bind _captureactors (LocalVariables|GetCaptureActors))
  (for _array_element (Variables|Default|GetCaptureClasses)
    (bind _outactors (Actor|GetAllActorsOfClass 0 _array_element))
    (Utilities|Array|AppendArray _captureactors _outactors))
  (return _captureactors))

(fn GetOutputDirectory
  (bind _returnvalue (Utilities|Paths|GetProjectDirectory))
  (bind _output_directory (Variables|IconCapture|GetOutput%20Directory))
  (bind _returnvalue_2 (Utilities|String|Append _returnvalue _output_directory "/"))
  (if (Variables|IconCapture|GetCreate%20Set%20Subdirectories)
    (return (Utilities|String|Replace (Utilities|String|Append _returnvalue_2 (Variables|Transient|GetOutputSubdirectory) "/") "//" "/"))
    (else
      (return (Utilities|String|Replace _returnvalue_2 "//" "/")))))

(fn GetWorldSettings
  (bind _outactors (Actor|GetAllActorsOfClass 0 "/Script/FortniteGame.FortWorldSettings"))
  (return (Utilities|Array|Get(acopy) _outactors)))

(fn HideAllBut
  (params (Item :type Actor%20Object%20Reference) (ItemList :type Array%20of%20Actor%20Object%20References))
  (for _array_element ItemList
    (Rendering|SetActorHiddenInGame _array_element true))
  (Utilities|IsValid Item
    (:Is%20Valid
      (Rendering|SetActorHiddenInGame Item false))))

(fn RemoveDuplicates
  (params (ItemList :type Array%20of%20Actor%20Object%20References))
  (bind _classset (LocalVariables|GetClassSet))
  (bind _uniqueitems (LocalVariables|GetUniqueItems))
  (for _array_element ItemList
    (if (not (Utilities|Set|ContainsItem _classset (Utilities|GetClass _array_element)))
      (Utilities|Set|Add _classset (Utilities|GetClass _array_element))
      (Utilities|Array|Add _uniqueitems _array_element)))
  (return _uniqueitems))

(fn SaveScreenshot
  (params (Filename :type String))
  (bind _outputdir (CallFunction|GetOutputDirectory))
  (Class|PlaysetCaptureLibrary|SaveScreenshot (Utilities|String|Append _outputdir Filename (Variables|IconCapture|GetImageSuffix)) (Math|Vector2D|MakeVector2D (Math|Conversions|ToFloat(Integer) (Variables|IconCapture|GetResolution)) (Math|Conversions|ToFloat(Integer) (Variables|IconCapture|GetResolution)))))

(fn ScreenshotSingleItem
  (params (Actor :type Actor%20Object%20Reference))
  (bind _returnvalue (Utilities|GetClass Actor))
  (bind _returnvalue_2 (Utilities|GetClassDisplayName _returnvalue))
  (if (Utilities|String|EndsWith _returnvalue_2 "_C")
    (CallFunction|SaveScreenshot (Utilities|String|LeftChop _returnvalue_2 2))
    (else
      (CallFunction|SaveScreenshot _returnvalue_2))))

(fn Setup
  (for _array_element (Variables|Default|GetConsoleCommands)
    (Development|ExecuteConsoleCommand 0 _array_element (Game|GetPlayerController 0))
    (bind _actors (CallFunction|GetAllCaptureActors))
    (bind _output_get (Variables|Transient|SetCaptureItems _actors))
    (for _array_element_2 _output_get
      (CallFunction|EnableCustomDepth _array_element_2 true))
    (bind _worldsettings (CallFunction|GetWorldSettings))
    (bind _returnvalue (EditorOnly|GetTimeOfDayManager _worldsettings))
    (TimeOfDay|SetTimeOfDaySpeed _returnvalue 0.0 :!class /Script/FortniteGame.FortTimeOfDayManagerInterface)
    (TimeOfDay|SetTimeOfDay _returnvalue (Variables|IconCapture|GetHour%20Of%20Day) :!class /Script/FortniteGame.FortTimeOfDayManagerInterface)))

(fn ConstructionScript)

(fn UseSceneCamera
  (bind _outactors (Actor|GetAllActorsOfClass 0 "/Game/ArtTools/PlaysetIconSceneCamera.PlaysetIconSceneCamera_C"))
  (if (> (Utilities|Array|Length _outactors) 0)
    (Transformation|SetActorTransform (Transformation|GetActorTransform (Utilities|Array|Get(acopy) _outactors)) false true)
    (Camera|SetFieldOfView (Variables|CameraActor|GetCameraComponent) (* (Variables|/Script/Engine.CameraComponent|GetFieldOfView (Variables|/Script/Engine.CameraActor|GetCameraComponent (Utilities|Array|Get(acopy) _outactors))) (Variables|IconCapture|GetFOV%20multiplier)))))
