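; Art-tool actor that previews and renders PNG icons for Fortnite playset building pieces.
; PopulatePlayset gathers BuildingSMActors, keeps one per class, hides them and auto-frames each from its bounds, FOV and fill percentage.
; Next/previous, yaw, pitch and axis functions edit per-piece orbit framing, which drives the camera and can be saved or undone.
; Capture functions copy the camera to a SceneCapture2D and export render targets; whole-playset capture is WIP.
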(event EventTick
  (params (DeltaSeconds :type Float%20%28single-precision%29)))

(event Collision|EventActorBeginOverlap
  (params (OtherActor :type Actor%20Object%20Reference)))

(event Custom|TargetData%20refresh%20using%20Target%20Index
  (Variables|PlaysetCapture|SetFramingData (Utilities|Array|Get(acopy) (Variables|PlaysetIconsDebug|GetPieces%20Collection) (Variables|PlaysetIconsDebug|GetTargetIndex))))

(event Custom|Target%20Data%20set%20on%20Piece%20using%20Target%20Index
  (Utilities|Array|SetArrayElem (Variables|PlaysetIconsDebug|GetPieces%20Collection) (Variables|PlaysetIconsDebug|GetTargetIndex) (Variables|PlaysetCapture|GetFramingData)))

(event Custom|Pause
  (Utilities|FlowControl|Delay 0.2))

(event Custom|Lights_Init
  (bind _lightpositionarray (Variables|Default|GetLightPositionArray))
  ;[0 0 0] initialize light array - not hooked up
  (Utilities|Array|Clear _lightpositionarray)
  (Utilities|Array|Add _lightpositionarray (Utilities|Struct|MakeLightandLocation 0 "92.000000,140.000000,113.000000" 20436.966797))
  (Utilities|Array|Add _lightpositionarray (Utilities|Struct|MakeLightandLocation 0 "-126.000000,28.000000,60.000000" 7135.34375))
  (Utilities|Array|Add _lightpositionarray (Utilities|Struct|MakeLightandLocation 0 "5.000000,63.000000,-48.000000" 3e+02))
  (Utilities|Array|Add _lightpositionarray (Utilities|Struct|MakeLightandLocation 0 "-20.000000,-107.000000,-10.000000" 1851.993774))
  (Utilities|Array|Add _lightpositionarray (Utilities|Struct|MakeLightandLocation 0 "96.000000,2.500000,-82.000000" 5e+03))
  (for _array_element (Variables|Default|GetLightPositionArray)
    (Transformation|SetRelativeLocation (Utilities|Struct|BreakLightandLocation _array_element) (.DefaultLocation_5_CF954B9145E2CB53183122B4A1BFA862 (Utilities|Struct|BreakLightandLocation _array_element)) false false)))

(fn 1:PopulatePlayset
  (bind _pieces_collection (Variables|PlaysetIconsDebug|GetPieces%20Collection))
  (bind _pieces_collection_2 (Variables|PlaysetIconsDebug|GetPieces%20Collection))
  (bind _playsetactorsset (Variables|Default|GetPlaysetActorsSet))
  (bind _framingdata (Variables|PlaysetCapture|GetFramingData))
  (bind _outactors (Actor|GetAllActorsOfClass "/Script/FortniteGame.BuildingSMActor"))
  (if (> (Utilities|Array|Length _outactors) 0)
    (Utilities|Array|Clear (Variables|PlaysetIconsDebug|GetPieces%20Collection))
    (Utilities|Array|Clear (Variables|PlaysetIconsDebug|GetArray%20Item))
    (for _array_element _outactors
      (Rendering|SetActorHiddenInGame _array_element true)
      (Variables|PlaysetCapture|SetFramingData (Utilities|Struct|MakePlaysetIconCaptureStructure _array_element "15.000000,-15.000000,800.000000" "0.000000,0.000000,0.000000" (Variables|PlaysetDefaults|GetFOV%20Default)))
      (PlaysetFraming|AutoFrame)
      (Utilities|Array|Add (Variables|PlaysetIconsDebug|GetArray%20Item) _framingdata)
      (if (not (Utilities|Set|ContainsItem _playsetactorsset (Utilities|GetClass _array_element)))
        (Utilities|Set|Add _playsetactorsset (Utilities|GetClass _array_element))
        (Utilities|Array|Add _pieces_collection _framingdata)))
    (Variables|PlaysetIconsDebug|SetPieces%20Collection _pieces_collection)
    (Variables|PlaysetIconsDebug|SetArray%20Item (Variables|PlaysetIconsDebug|GetArray%20Item))
    (bind _output_get (Variables|PlaysetIconsDebug|SetTargetIndex 0))
    (Variables|PlaysetIconsDebug|SetPieces%20Length (Utilities|Array|Length _pieces_collection_2))
    (Variables|PlaysetCapture|SetFramingData (Utilities|Array|Get(acopy) _pieces_collection_2 _output_get))
    (PlaysetDefaults|ActorVisible (Utilities|Struct|BreakPlaysetIconCaptureStructure (Utilities|Array|Get(acopy) _pieces_collection_2 _output_get)))
    (PlaysetDefaults|UpdateCamera)))

(fn 2:NextPiece
  (bind _pieces_collection (Variables|PlaysetIconsDebug|GetPieces%20Collection))
  (bind _returnvalue (Utilities|Array|Length _pieces_collection))
  (bind _targetindex (Variables|PlaysetIconsDebug|GetTargetIndex))
  (bind _returnvalue_2 (+ _targetindex 1))
  (bind _returnvalue_3 (Math|Integer|%25%28Integer%29 _returnvalue_2 _returnvalue))
  (bind _returnvalue_4 (Math|Integer|Absolute(Integer) _returnvalue_3))
  (PlaysetDefaults|ArrayInvisible)
  (Variables|PlaysetIconsDebug|SetTargetIndex _returnvalue_4)
  (bind _output_get (Variables|PlaysetCapture|SetFramingData (Utilities|Array|Get(acopy) _pieces_collection _returnvalue_4)))
  (PlaysetDefaults|ActorVisible (Utilities|Struct|BreakPlaysetIconCaptureStructure _output_get))
  (PlaysetDefaults|UpdateCamera))

(fn 3:PreviousPiece
  (bind _pieces_collection (Variables|PlaysetIconsDebug|GetPieces%20Collection))
  (bind _returnvalue (Utilities|Array|Length _pieces_collection))
  (bind _targetindex (Variables|PlaysetIconsDebug|GetTargetIndex))
  (bind _returnvalue_2 (+ _targetindex -1))
  (bind _returnvalue_3 (Math|Integer|%25%28Integer%29 _returnvalue_2 _returnvalue))
  (bind _returnvalue_4 (Math|Integer|Absolute(Integer) _returnvalue_3))
  (PlaysetDefaults|ArrayInvisible)
  (if (< _returnvalue_2 0)
    (Variables|PlaysetIconsDebug|SetTargetIndex (- _returnvalue 1))
    (else
      (Variables|PlaysetIconsDebug|SetTargetIndex _returnvalue_4)))
  (bind _output_get (Variables|PlaysetCapture|SetFramingData (Utilities|Array|Get(acopy) _pieces_collection _returnvalue_4)))
  (PlaysetDefaults|ActorVisible (Utilities|Struct|BreakPlaysetIconCaptureStructure _output_get))
  (PlaysetDefaults|UpdateCamera))

(fn 4:CapturePiece
  (bind _framingdata (Variables|PlaysetCapture|GetFramingData))
  (bind (_actorarray_26_7c8fd96d47b9eda61b2254844264b3f9 _cameraorbitposition_25_66c3d8c34eeeedbbe5cfa8bfeee4dd7f _lookatoffset_30_f54c94fd4357ff84fd3afbb5a0ad93dd _fov_28_9b39550b44b7991ec8d6cf964225bc07) (Utilities|Struct|BreakPlaysetIconCaptureStructure _framingdata))
  (bind _returnvalue (Math|Color|MakeColor 1.0))
  (bind _scenecapturecomponent2d (Variables|Default|GetSceneCaptureComponent2D))
  (bind _scenecapturecomponent2d_2 (Variables|Default|GetSceneCaptureComponent2D))
  (bind _texturetarget (Variables|/Script/Engine.SceneCaptureComponent2D|GetTextureTarget _scenecapturecomponent2d_2))
  (bind _camera (Variables|Default|GetCamera))
  (switch Utilities|FlowControl|Switch|SwitchonPlaysetIconCaptureResolutions (Variables|PlaysetDefaults|GetIconRenderResolution)
    (:NewEnumerator0
      (bind _output_get (Variables|/Script/Engine.SceneCaptureComponent2D|SetTextureTarget "/Game/ArtTools/PlaysetIconRT_256x256.PlaysetIconRT_256x256" _scenecapturecomponent2d))
      (Rendering|ClearRenderTarget2D _output_get _returnvalue))
    (:NewEnumerator1
      (bind _output_get_2 (Variables|/Script/Engine.SceneCaptureComponent2D|SetTextureTarget "/Game/ArtTools/PlaysetIconRT_512x512.PlaysetIconRT_512x512" _scenecapturecomponent2d))
      (Rendering|ClearRenderTarget2D _output_get_2 _returnvalue))
    (:NewEnumerator2
      (bind _output_get_3 (Variables|/Script/Engine.SceneCaptureComponent2D|SetTextureTarget "/Game/ArtTools/PlaysetIconRT_1024x1024.PlaysetIconRT_1024x1024" _scenecapturecomponent2d))
      (Rendering|ClearRenderTarget2D _output_get_3 _returnvalue))
    (:NewEnumerator3
      (bind _output_get_4 (Variables|/Script/Engine.SceneCaptureComponent2D|SetTextureTarget "/Game/ArtTools/PlaysetIconRT_2048x2048.PlaysetIconRT_2048x2048" _scenecapturecomponent2d))
      (Rendering|ClearRenderTarget2D _output_get_4 _returnvalue)))
  (PlaysetDefaults|UpdateCamera)
  (Components|Activation|SetActive _camera true false) ; not sure why this is needed
  (Transformation|SetWorldTransform (Variables|Default|GetSceneCaptureComponent2D) (Transformation|GetWorldTransform _camera) false false)
  (Variables|/Script/Engine.TextureRenderTarget|SetTargetGamma 2.2 _texturetarget)
  (bind (_filepath _returnvalue_2) (PlaysetDefaults|BuildFileNumbering (Variables|PlaysetIconsDebug|GetTargetIndex) _actorarray_26_7c8fd96d47b9eda61b2254844264b3f9 (Variables|PlaysetDefaults|GetUse%20Object%20Name%20as%20Filename)))
  (Variables|/Script/Engine.SceneCaptureComponent2D|SetFOVAngle _fov_28_9b39550b44b7991ec8d6cf964225bc07 _scenecapturecomponent2d_2)
  (Rendering|SceneCapture|CaptureScene _scenecapturecomponent2d_2 :!class /Script/Engine.SceneCaptureComponent2D)
  (Rendering|ExportRenderTarget _texturetarget _filepath _returnvalue_2))

(fn 5:CapturePieces
  (bind _pieces_collection (Variables|PlaysetIconsDebug|GetPieces%20Collection))
  (Rendering|SetActorHiddenInGame (Utilities|Struct|BreakPlaysetIconCaptureStructure (Utilities|Array|Get(acopy) (Variables|PlaysetIconsDebug|GetPieces%20Collection) (Variables|PlaysetIconsDebug|GetTargetIndex))) true)
  (for _array_element _pieces_collection
    (bind _output_get (Variables|PlaysetCapture|SetFramingData _array_element))
    (Rendering|SetActorHiddenInGame (Utilities|Struct|BreakPlaysetIconCaptureStructure _output_get) false)
    (PlaysetCapture|4:CapturePiece)
    (CallFunction|Pause)
    (Rendering|SetActorHiddenInGame (Utilities|Struct|BreakPlaysetIconCaptureStructure _output_get) true))
  (bind _output_get_2 (Variables|PlaysetCapture|SetFramingData (Utilities|Array|Get(acopy) _pieces_collection (Variables|PlaysetIconsDebug|GetTargetIndex))))
  (Rendering|SetActorHiddenInGame (Utilities|Struct|BreakPlaysetIconCaptureStructure _output_get_2) false)
  (PlaysetDefaults|UpdateCamera))

(fn ActorInvisible
  (params (Actor :type Actor%20Object%20Reference))
  (Rendering|SetActorHiddenInGame Actor true))

(fn ActorVisible
  (params (Actor :type Actor%20Object%20Reference))
  (Rendering|SetActorHiddenInGame Actor false))

(fn ArrayInvisible
  (bind _array_item (Variables|PlaysetIconsDebug|GetArray%20Item))
  (for (_ _array_index) _array_item
    (Rendering|SetActorHiddenInGame (Utilities|Struct|BreakPlaysetIconCaptureStructure (Utilities|Array|Get(aref) _array_item _array_index)) true)))

(fn ArrayVisible
  (bind _array_item (Variables|PlaysetIconsDebug|GetArray%20Item))
  (for (_ _array_index) _array_item
    (Rendering|SetActorHiddenInGame (Utilities|Struct|BreakPlaysetIconCaptureStructure (Utilities|Array|Get(aref) _array_item _array_index)) false)))

(fn AutoFrame
  (bind (_actorarray_26_7c8fd96d47b9eda61b2254844264b3f9 _cameraorbitposition_25_66c3d8c34eeeedbbe5cfa8bfeee4dd7f _lookatoffset_30_f54c94fd4357ff84fd3afbb5a0ad93dd _fov_28_9b39550b44b7991ec8d6cf964225bc07) (Utilities|Struct|BreakPlaysetIconCaptureStructure (Variables|PlaysetCapture|GetFramingData)))
  (bind (_origin _boxextent _sphereradius) (Collision|GetComponentBounds (Variables|/Script/Engine.Actor|GetRootComponent _actorarray_26_7c8fd96d47b9eda61b2254844264b3f9)))
  (Variables|PlaysetCapture|SetFramingData (Utilities|Struct|MakePlaysetIconCaptureStructure _actorarray_26_7c8fd96d47b9eda61b2254844264b3f9 (Math|Vector|MakeVector (.x _cameraorbitposition_25_66c3d8c34eeeedbbe5cfa8bfeee4dd7f) (.y _cameraorbitposition_25_66c3d8c34eeeedbbe5cfa8bfeee4dd7f) (/ (* _sphereradius (Utilities|Reciprocal (* (Variables|PlaysetDefaults|GetIcon%20Fill%20Percentage) 0.01))) (Math|Trig|Tan(Degrees) (/ _fov_28_9b39550b44b7991ec8d6cf964225bc07 2.0)))) "0.000000,0.000000,0.000000" _fov_28_9b39550b44b7991ec8d6cf964225bc07))
  (PlaysetDefaults|UpdateCamera)
  (return))

(fn AxisX+
  (bind (_actorarray_26_7c8fd96d47b9eda61b2254844264b3f9 _cameraorbitposition_25_66c3d8c34eeeedbbe5cfa8bfeee4dd7f _lookatoffset_30_f54c94fd4357ff84fd3afbb5a0ad93dd _fov_28_9b39550b44b7991ec8d6cf964225bc07) (Utilities|Struct|BreakPlaysetIconCaptureStructure (Variables|PlaysetCapture|GetFramingData)))
  (Variables|PlaysetCapture|SetFramingData (Utilities|Struct|MakePlaysetIconCaptureStructure _actorarray_26_7c8fd96d47b9eda61b2254844264b3f9 (Math|Vector|MakeVector 0.0 (.y _cameraorbitposition_25_66c3d8c34eeeedbbe5cfa8bfeee4dd7f) (.z _cameraorbitposition_25_66c3d8c34eeeedbbe5cfa8bfeee4dd7f)) _lookatoffset_30_f54c94fd4357ff84fd3afbb5a0ad93dd _fov_28_9b39550b44b7991ec8d6cf964225bc07))
  (PlaysetDefaults|UpdateCamera))

(fn AxisX-
  (bind (_actorarray_26_7c8fd96d47b9eda61b2254844264b3f9 _cameraorbitposition_25_66c3d8c34eeeedbbe5cfa8bfeee4dd7f _lookatoffset_30_f54c94fd4357ff84fd3afbb5a0ad93dd _fov_28_9b39550b44b7991ec8d6cf964225bc07) (Utilities|Struct|BreakPlaysetIconCaptureStructure (Variables|PlaysetCapture|GetFramingData)))
  (Variables|PlaysetCapture|SetFramingData (Utilities|Struct|MakePlaysetIconCaptureStructure _actorarray_26_7c8fd96d47b9eda61b2254844264b3f9 (Math|Vector|MakeVector 1.8e+02 (.y _cameraorbitposition_25_66c3d8c34eeeedbbe5cfa8bfeee4dd7f) (.z _cameraorbitposition_25_66c3d8c34eeeedbbe5cfa8bfeee4dd7f)) _lookatoffset_30_f54c94fd4357ff84fd3afbb5a0ad93dd _fov_28_9b39550b44b7991ec8d6cf964225bc07))
  (PlaysetDefaults|UpdateCamera))

(fn AxisY+
  (bind (_actorarray_26_7c8fd96d47b9eda61b2254844264b3f9 _cameraorbitposition_25_66c3d8c34eeeedbbe5cfa8bfeee4dd7f _lookatoffset_30_f54c94fd4357ff84fd3afbb5a0ad93dd _fov_28_9b39550b44b7991ec8d6cf964225bc07) (Utilities|Struct|BreakPlaysetIconCaptureStructure (Variables|PlaysetCapture|GetFramingData)))
  (Variables|PlaysetCapture|SetFramingData (Utilities|Struct|MakePlaysetIconCaptureStructure _actorarray_26_7c8fd96d47b9eda61b2254844264b3f9 (Math|Vector|MakeVector 9e+01 (.y _cameraorbitposition_25_66c3d8c34eeeedbbe5cfa8bfeee4dd7f) (.z _cameraorbitposition_25_66c3d8c34eeeedbbe5cfa8bfeee4dd7f)) _lookatoffset_30_f54c94fd4357ff84fd3afbb5a0ad93dd _fov_28_9b39550b44b7991ec8d6cf964225bc07))
  (PlaysetDefaults|UpdateCamera))

(fn AxisY-
  (bind (_actorarray_26_7c8fd96d47b9eda61b2254844264b3f9 _cameraorbitposition_25_66c3d8c34eeeedbbe5cfa8bfeee4dd7f _lookatoffset_30_f54c94fd4357ff84fd3afbb5a0ad93dd _fov_28_9b39550b44b7991ec8d6cf964225bc07) (Utilities|Struct|BreakPlaysetIconCaptureStructure (Variables|PlaysetCapture|GetFramingData)))
  (Variables|PlaysetCapture|SetFramingData (Utilities|Struct|MakePlaysetIconCaptureStructure _actorarray_26_7c8fd96d47b9eda61b2254844264b3f9 (Math|Vector|MakeVector -9e+01 (.y _cameraorbitposition_25_66c3d8c34eeeedbbe5cfa8bfeee4dd7f) (.z _cameraorbitposition_25_66c3d8c34eeeedbbe5cfa8bfeee4dd7f)) _lookatoffset_30_f54c94fd4357ff84fd3afbb5a0ad93dd _fov_28_9b39550b44b7991ec8d6cf964225bc07))
  (PlaysetDefaults|UpdateCamera))

(fn BuildFileNumbering
  (params (InInt :type Integer) (Component :type Actor%20Object%20Reference) (Use%20Object%20Name :type Boolean))
  (if Use%20Object%20Name
    (return (Variables|PlaysetDefaults|GetFilePath) (Utilities|String|Append (Utilities|GetDisplayName Component) ".png"))
    (else
      (return (Variables|PlaysetDefaults|GetFilePath) (Utilities|String|Append (Variables|PlaysetDefaults|GetFileName) (Utilities|String|ToString(Integer) InInt) ".png")))))

(fn LookatTargetPositiveX
  (bind _framingdata (Variables|PlaysetCapture|GetFramingData))
  (bind _actorarray_26_7c8fd96d47b9eda61b2254844264b3f9 (Utilities|Struct|BreakPlaysetIconCaptureStructure _framingdata))
  (bind (_origin _boxextent) (Collision|GetActorBounds _actorarray_26_7c8fd96d47b9eda61b2254844264b3f9 false false))
  (Transformation|SetWorldLocation (Variables|Default|GetProxy_CamPosition) _origin false false)
  (Transformation|SetRelativeScale3D (Variables|Default|GetProxy_CamPosition) (* _boxextent 0.02))
  (Variables|PlaysetIconsDebug|Setdebug%20Object%20Extents _boxextent)
  (Variables|PlaysetIconsDebug|Setdebug%20Object%20Centroid _origin))

(fn Pitch+15
  (bind (_actorarray_26_7c8fd96d47b9eda61b2254844264b3f9 _cameraorbitposition_25_66c3d8c34eeeedbbe5cfa8bfeee4dd7f _lookatoffset_30_f54c94fd4357ff84fd3afbb5a0ad93dd _fov_28_9b39550b44b7991ec8d6cf964225bc07) (Utilities|Struct|BreakPlaysetIconCaptureStructure (Variables|PlaysetCapture|GetFramingData)))
  (Variables|PlaysetCapture|SetFramingData (Utilities|Struct|MakePlaysetIconCaptureStructure _actorarray_26_7c8fd96d47b9eda61b2254844264b3f9 (Math|Vector|MakeVector (.x _cameraorbitposition_25_66c3d8c34eeeedbbe5cfa8bfeee4dd7f) (+ (.y _cameraorbitposition_25_66c3d8c34eeeedbbe5cfa8bfeee4dd7f) -15.0) (.z _cameraorbitposition_25_66c3d8c34eeeedbbe5cfa8bfeee4dd7f)) "0.000000,0.000000,0.000000" _fov_28_9b39550b44b7991ec8d6cf964225bc07))
  (PlaysetDefaults|UpdateCamera))

(fn Pitch-15
  (bind (_actorarray_26_7c8fd96d47b9eda61b2254844264b3f9 _cameraorbitposition_25_66c3d8c34eeeedbbe5cfa8bfeee4dd7f _lookatoffset_30_f54c94fd4357ff84fd3afbb5a0ad93dd _fov_28_9b39550b44b7991ec8d6cf964225bc07) (Utilities|Struct|BreakPlaysetIconCaptureStructure (Variables|PlaysetCapture|GetFramingData)))
  (Variables|PlaysetCapture|SetFramingData (Utilities|Struct|MakePlaysetIconCaptureStructure _actorarray_26_7c8fd96d47b9eda61b2254844264b3f9 (Math|Vector|MakeVector (.x _cameraorbitposition_25_66c3d8c34eeeedbbe5cfa8bfeee4dd7f) (+ (.y _cameraorbitposition_25_66c3d8c34eeeedbbe5cfa8bfeee4dd7f) 15.0) (.z _cameraorbitposition_25_66c3d8c34eeeedbbe5cfa8bfeee4dd7f)) _lookatoffset_30_f54c94fd4357ff84fd3afbb5a0ad93dd _fov_28_9b39550b44b7991ec8d6cf964225bc07))
  (PlaysetDefaults|UpdateCamera))

(fn SaveFraming
  (Utilities|Array|SetArrayElem (Variables|PlaysetIconsDebug|GetPieces%20Collection) (Variables|PlaysetIconsDebug|GetTargetIndex) (Variables|PlaysetCapture|GetFramingData)))

(fn UndoFraming
  (Variables|PlaysetCapture|SetFramingData (Utilities|Array|Get(acopy) (Variables|PlaysetIconsDebug|GetPieces%20Collection) (Variables|PlaysetIconsDebug|GetTargetIndex)))
  (PlaysetDefaults|UpdateCamera))

(fn UpdateCamera
  (bind _framingdata (Variables|PlaysetCapture|GetFramingData))
  (bind (_actorarray_26_7c8fd96d47b9eda61b2254844264b3f9 _cameraorbitposition_25_66c3d8c34eeeedbbe5cfa8bfeee4dd7f _lookatoffset_30_f54c94fd4357ff84fd3afbb5a0ad93dd _fov_28_9b39550b44b7991ec8d6cf964225bc07) (Utilities|Struct|BreakPlaysetIconCaptureStructure _framingdata))
  (bind _z (.z _cameraorbitposition_25_66c3d8c34eeeedbbe5cfa8bfeee4dd7f))
  (bind _y (.y _cameraorbitposition_25_66c3d8c34eeeedbbe5cfa8bfeee4dd7f))
  (bind _x (.x _cameraorbitposition_25_66c3d8c34eeeedbbe5cfa8bfeee4dd7f))
  (bind _returnvalue (Math|Vector|CreateVectorfromYawPitch _x _y _z))
  (bind _rootcomponent (Variables|/Script/Engine.Actor|GetRootComponent _actorarray_26_7c8fd96d47b9eda61b2254844264b3f9))
  (bind _origin (Collision|GetComponentBounds _rootcomponent))
  (bind _returnvalue_2 (+ _origin _lookatoffset_30_f54c94fd4357ff84fd3afbb5a0ad93dd))
  (bind _returnvalue_3 (- _returnvalue_2 _returnvalue))
  (bind _proxy_camposition (Variables|Default|GetProxy_CamPosition))
  (bind _camera (Variables|Default|GetCamera))
  (Transformation|SetWorldLocation _camera _returnvalue_3 false false)
  (Transformation|SetWorldRotation _camera (Math|Rotator|FindLookatRotation _returnvalue_3 _returnvalue_2) false false)
  (Camera|SetFieldOfView (Variables|Default|GetCamera) _fov_28_9b39550b44b7991ec8d6cf964225bc07)
  (Transformation|SetWorldLocation _proxy_camposition _returnvalue_3 false false)
  (Transformation|SetWorldLocation (Variables|Default|GetProxy_CamTarget) _returnvalue_2 false false)
  (Development|SetHiddeninGame _proxy_camposition (Variables|PlaysetDefaults|GetCamera%20Proxies%20Hidden%20in%20Game) false)
  (Rendering|SetVisibility (Variables|Default|GetProxy_CamPosition) (Variables|PlaysetDefaults|GetCamera%20Proxies%20Visible%20in%20Editor) false))

(fn UpdateCameraPlayset
  (bind _playset_bounds (Variables|Default|GetPlaySet%20Bounds))
  (bind _x (.x _playset_bounds))
  (bind _y (.y _playset_bounds))
  (bind _returnvalue (* _y -1.5))
  (bind _returnvalue_2 (* _x -1.5))
  (bind _returnvalue_3 (Math|Vector|MakeVector _returnvalue_2 _returnvalue))
  (bind _playset_center (Variables|PlaysetDefaults|GetPlaySet%20Center))
  (bind _returnvalue_4 (+ _playset_center _returnvalue_3))
  (bind _camera (Variables|Default|GetCamera))
  (bind _proxy_camposition (Variables|Default|GetProxy_CamPosition))
  (Transformation|SetWorldLocation _camera _returnvalue_4 false false)
  (Transformation|SetWorldRotation _camera (Math|Rotator|FindLookatRotation _returnvalue_4 _playset_center) false false)
  (Transformation|SetWorldLocation _proxy_camposition "0, 0, 0" false false)
  (Transformation|SetWorldLocation (Variables|Default|GetProxy_CamTarget) "0, 0, 0" false false)
  (Development|SetHiddeninGame _proxy_camposition (Variables|PlaysetDefaults|GetCamera%20Proxies%20Hidden%20in%20Game) false))

(fn ConstructionScript
  (bind _output_get (Variables|PlaysetIconsDebug|SetPieces%20Length (Utilities|Array|Length (Variables|PlaysetIconsDebug|GetPieces%20Collection))))
  (if (< _output_get 1)
    _
    (elif (< _output_get (Variables|PlaysetIconsDebug|GetTargetIndex))
      _
      (else
        (PlaysetDefaults|UpdateCamera))))
  (Variables|PlaysetIconsDebug|SetTargetIndex 0)
  (CallFunction|TargetDataRefreshUsingTargetIndex))

(fn WIP-PopulateGenericUnique
  (bind _pieces_collection (Variables|PlaysetIconsDebug|GetPieces%20Collection))
  (bind _pieces_collection_2 (Variables|PlaysetIconsDebug|GetPieces%20Collection))
  (bind _playsetactorsset (Variables|Default|GetPlaysetActorsSet))
  (bind _framingdata (Variables|PlaysetCapture|GetFramingData))
  ;[1 0.4375 0.4375] WIP
  (bind _outactors (Actor|GetAllActorsOfClass "/Script/FortniteGame.BuildingActor"))
  (if (> (Utilities|Array|Length _outactors) 0)
    (Utilities|Array|Clear (Variables|PlaysetIconsDebug|GetPieces%20Collection))
    (Utilities|Array|Clear (Variables|PlaysetIconsDebug|GetArray%20Item))
    (for _array_element _outactors
      (bind _asbuilding_wall (Utilities|Casting|CastToBuildingWall _array_element)
        (:then
          (Building|GetBuildingTextureData _asbuilding_wall :!class /Script/FortniteGame.BuildingSMActor))
        (:CastFailed
          (bind _asbuilding_floor (Utilities|Casting|CastToBuildingFloor _array_element)
            (:then
              (Building|GetBuildingTextureData _asbuilding_floor :!class /Script/FortniteGame.BuildingSMActor))
            (:CastFailed
              (bind _asbuilding_stairs (Utilities|Casting|CastToBuildingStairs _array_element)
                (:then
                  (Building|GetBuildingTextureData _asbuilding_stairs :!class /Script/FortniteGame.BuildingSMActor))
                (:CastFailed
                  (bind _asbuilding_roof (Utilities|Casting|CastToBuildingRoof _array_element)
                    (:then
                      (Building|GetBuildingTextureData _asbuilding_roof :!class /Script/FortniteGame.BuildingSMActor))
                    (:CastFailed
                      (bind _asbuilding_static_mesh (Utilities|Casting|CastToBuildingSMActor _array_element)
                        (:then
                          (Building|GetBuildingTextureData _asbuilding_static_mesh :!class /Script/FortniteGame.BuildingSMActor))
                        (:CastFailed
                          _))))))))))
      (Rendering|SetActorHiddenInGame _array_element true)
      (Variables|PlaysetCapture|SetFramingData (Utilities|Struct|MakePlaysetIconCaptureStructure _array_element "15.000000,-15.000000,800.000000" "0.000000,0.000000,0.000000" (Variables|PlaysetDefaults|GetFOV%20Default)))
      (PlaysetFraming|AutoFrame)
      (Utilities|Array|Add (Variables|PlaysetIconsDebug|GetArray%20Item) _framingdata)
      (if (not (Utilities|Set|ContainsItem _playsetactorsset (Utilities|GetClass _array_element)))
        (Utilities|Set|Add _playsetactorsset (Utilities|GetClass _array_element))
        (Utilities|Array|Add _pieces_collection _framingdata)))
    (Variables|PlaysetIconsDebug|SetPieces%20Collection _pieces_collection)
    (Variables|PlaysetIconsDebug|SetArray%20Item (Variables|PlaysetIconsDebug|GetArray%20Item))
    (bind _output_get (Variables|PlaysetIconsDebug|SetTargetIndex 0))
    (Variables|PlaysetIconsDebug|SetPieces%20Length (Utilities|Array|Length _pieces_collection_2))
    (Variables|PlaysetCapture|SetFramingData (Utilities|Array|Get(acopy) _pieces_collection_2 _output_get))
    (PlaysetDefaults|ActorVisible (Utilities|Struct|BreakPlaysetIconCaptureStructure (Utilities|Array|Get(acopy) _pieces_collection_2 _output_get)))
    (PlaysetDefaults|UpdateCamera)))

(fn WIP6:CapturePlayset
  (bind _scenecapturecomponent2d (Variables|Default|GetSceneCaptureComponent2D))
  (bind _texturetarget (Variables|/Script/Engine.SceneCaptureComponent2D|GetTextureTarget _scenecapturecomponent2d))
  (bind _actorboundsarray (Variables|PlaysetDefaults|GetActorBoundsArray))
  (for _array_element (Variables|PlaysetIconsDebug|GetArray%20Item)
    (Utilities|Array|Add _actorboundsarray (Utilities|Struct|BreakPlaysetIconCaptureStructure _array_element)))
  (bind _output_get (Variables|PlaysetDefaults|SetActorBoundsArray _actorboundsarray))
  (bind (_center _boxextent) (Collision|GetActorArrayBounds _output_get))
  (Variables|PlaysetDefaults|SetPlaySet%20Center _center)
  (Variables|Default|SetPlaySet%20Bounds _boxextent)
  (PlaysetDefaults|UpdateCameraPlayset)
  (PlaysetDefaults|ArrayVisible)
  (Transformation|SetWorldTransform (Variables|Default|GetSceneCaptureComponent2D) (Transformation|GetWorldTransform (Variables|Default|GetCamera)) false false)
  (Rendering|SceneCapture|CaptureScene _scenecapturecomponent2d :!class /Script/Engine.SceneCaptureComponent2D)
  (bind (_filepath _returnvalue) (PlaysetDefaults|BuildFileNumbering (Variables|PlaysetIconsDebug|GetTargetIndex) (Variables|PlaysetDefaults|GetTargetActor) (Variables|PlaysetDefaults|GetUse%20Object%20Name%20as%20Filename)))
  (Variables|/Script/Engine.TextureRenderTarget|SetTargetGamma 2.0 _texturetarget)
  (Rendering|ExportRenderTarget _texturetarget _filepath (Variables|PlaysetDefaults|GetplaySetName))
  (PlaysetDefaults|ArrayInvisible))

(fn Yaw+15
  (bind (_actorarray_26_7c8fd96d47b9eda61b2254844264b3f9 _cameraorbitposition_25_66c3d8c34eeeedbbe5cfa8bfeee4dd7f _lookatoffset_30_f54c94fd4357ff84fd3afbb5a0ad93dd _fov_28_9b39550b44b7991ec8d6cf964225bc07) (Utilities|Struct|BreakPlaysetIconCaptureStructure (Variables|PlaysetCapture|GetFramingData)))
  (Variables|PlaysetCapture|SetFramingData (Utilities|Struct|MakePlaysetIconCaptureStructure _actorarray_26_7c8fd96d47b9eda61b2254844264b3f9 (Math|Vector|MakeVector (+ (.x _cameraorbitposition_25_66c3d8c34eeeedbbe5cfa8bfeee4dd7f) 15.0) (.y _cameraorbitposition_25_66c3d8c34eeeedbbe5cfa8bfeee4dd7f) (.z _cameraorbitposition_25_66c3d8c34eeeedbbe5cfa8bfeee4dd7f)) _lookatoffset_30_f54c94fd4357ff84fd3afbb5a0ad93dd _fov_28_9b39550b44b7991ec8d6cf964225bc07))
  (PlaysetDefaults|UpdateCamera))

(fn Yaw-15
  (bind (_actorarray_26_7c8fd96d47b9eda61b2254844264b3f9 _cameraorbitposition_25_66c3d8c34eeeedbbe5cfa8bfeee4dd7f _lookatoffset_30_f54c94fd4357ff84fd3afbb5a0ad93dd _fov_28_9b39550b44b7991ec8d6cf964225bc07) (Utilities|Struct|BreakPlaysetIconCaptureStructure (Variables|PlaysetCapture|GetFramingData)))
  (Variables|PlaysetCapture|SetFramingData (Utilities|Struct|MakePlaysetIconCaptureStructure _actorarray_26_7c8fd96d47b9eda61b2254844264b3f9 (Math|Vector|MakeVector (+ (.x _cameraorbitposition_25_66c3d8c34eeeedbbe5cfa8bfeee4dd7f) -15.0) (.y _cameraorbitposition_25_66c3d8c34eeeedbbe5cfa8bfeee4dd7f) (.z _cameraorbitposition_25_66c3d8c34eeeedbbe5cfa8bfeee4dd7f)) _lookatoffset_30_f54c94fd4357ff84fd3afbb5a0ad93dd _fov_28_9b39550b44b7991ec8d6cf964225bc07))
  (PlaysetDefaults|UpdateCamera))
