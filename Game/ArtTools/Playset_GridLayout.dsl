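; Art tool that collects Fortnite BuildingSMActors in the level and lays them out in a grid for playsets.
; Gathering records props by unique class and actors by class, mesh and texture-data key, storing bounds and volume.
; StructureBubbleSort compares X extent despite volume labels; PlaceObjectsinaGrid then positions actors using fixed or bounds-derived spacing, optional square layout and split prop offset.
; Helper events fit a floor, swap positions and rotate actors.
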(event EventBeginPlay)

(event Collision|EventActorBeginOverlap
  (params (OtherActor :type Actor%20Object%20Reference)))

(event Custom|Floor%20Replace
  (bind _floor_object_bp (Variables|Default|GetFloor%20Object%20BP))
  (if (not (Class|ItemUIFunctionLibrary|IsNullItem _floor_object_bp))
    (Transformation|SetActorTransform (Variables|Default|GetTarget_Actor) (Math|Transform|MakeTransform (Transformation|GetWorldLocation (Variables|/Script/Engine.StaticMeshActor|GetStaticMeshComponent _floor_object_bp)) (Transformation|GetWorldRotation (Variables|/Script/Engine.StaticMeshActor|GetStaticMeshComponent _floor_object_bp)) (Transformation|GetWorldScale (Variables|/Script/Engine.StaticMeshActor|GetStaticMeshComponent _floor_object_bp))) false false)))

(event EventTick
  (params (DeltaSeconds :type Float%20%28single-precision%29)))

(event Custom|WIP%20Actor%20Details
  (bind _target_actor (Variables|Default|GetTarget_Actor))
  (bind _rootcomponent (Variables|/Script/Engine.Actor|GetRootComponent _target_actor))
  (bind (_origin _boxextent) (Collision|GetComponentBounds _rootcomponent))
  (Variables|Default|SetTarget%20Actor%20Origin _origin)
  (Variables|Default|SetTarget%20Actor%20Bounds _boxextent)
  (Variables|Default|SetTarget%20Actor%20Position (Transformation|GetWorldLocation _rootcomponent)))

(event Custom|Gather%20Unique%20Objects
  (bind _objectstatsarray (Variables|PlaysetGridLayout|GetObjectStatsArray))
  (PlaysetGridLayout|ClearArrays)
  (bind _outactors (Actor|GetAllActorsOfClass "/Script/FortniteGame.BuildingSMActor"))
  (for _array_element _outactors
    (Variables|Default|SetObjectClassName "")
    (Utilities|Casting|CastToBuildingWall _array_element
      (:then
        (Variables|Default|SetObjectClassName "BuildingWall"))
      (:CastFailed
        (Utilities|Casting|CastToBuildingFloor _array_element
          (:then
            (Variables|Default|SetObjectClassName "BuildingFloor"))
          (:CastFailed
            (Utilities|Casting|CastToBuildingStairs _array_element
              (:then
                (Variables|Default|SetObjectClassName "BuildingStairs"))
              (:CastFailed
                (Utilities|Casting|CastToBuildingRoof _array_element
                  (:then
                    (Variables|Default|SetObjectClassName "BuildingRoof"))
                  (:CastFailed
                    (if (not (Utilities|Set|ContainsItem (Variables|Default|GetPlayset%20Actors%20Set) (Utilities|GetClass _array_element)))
                      (Utilities|Set|Add (Variables|Default|GetPlayset%20Actors%20Set) (Utilities|GetClass _array_element))
                      (Development|PrintString (Utilities|String|ToString(Float) (* (* (* (.x (* (.BoxExtent (Collision|GetComponentBounds (Variables|/Script/Engine.Actor|GetRootComponent _array_element))) 0.01)) (.y (* (.BoxExtent (Collision|GetComponentBounds (Variables|/Script/Engine.Actor|GetRootComponent _array_element))) 0.01))) (.z (* (.BoxExtent (Collision|GetComponentBounds (Variables|/Script/Engine.Actor|GetRootComponent _array_element))) 0.01))) 8.0)))
                      (Utilities|Array|Add (Variables|PlaysetGridLayout|GetObjectStatsArray) (Utilities|Struct|MakePlaysetGridLayoutStruc _array_element (.BoxExtent (Collision|GetComponentBounds (Variables|/Script/Engine.Actor|GetRootComponent _array_element))) (* (* (* (.x (* (.BoxExtent (Collision|GetComponentBounds (Variables|/Script/Engine.Actor|GetRootComponent _array_element))) 0.01)) (.y (* (.BoxExtent (Collision|GetComponentBounds (Variables|/Script/Engine.Actor|GetRootComponent _array_element))) 0.01))) (.z (* (.BoxExtent (Collision|GetComponentBounds (Variables|/Script/Engine.Actor|GetRootComponent _array_element))) 0.01))) 8.0) true)))))))))))
    (Variables|Default|SetObjectMatName "")
    (bind _outtexturedataarray (Building|GetBuildingTextureData _array_element :!class /Script/FortniteGame.BuildingSMActor))
    (for _array_element_2 _outtexturedataarray
      (Development|PrintString "Hello")
      (if (> (Utilities|String|Len (Variables|Default|GetObjectMatName)) 0)
        (Variables|Default|SetObjectMatName (Utilities|String|Append (Variables|Default|GetObjectMatName) ":" (Utilities|GetDisplayName _array_element_2) ""))
        (else
          (Variables|Default|SetObjectMatName (Utilities|GetDisplayName _array_element_2)))))
    (Utilities|Map|Add (Variables|PlaysetGridLayout|GetMaterialDictionary) (Utilities|String|Append (Variables|Default|GetObjectClassName) "-" (Utilities|GetDisplayName (Variables|/Script/Engine.StaticMeshComponent|GetStaticMesh (Variables|/Script/FortniteGame.BuildingSMActor|GetStaticMeshComponent _array_element))) "-" (Variables|Default|GetObjectMatName)) _array_element))
  (bind _values (Utilities|Map|Values (Variables|PlaysetGridLayout|GetMaterialDictionary)))
  (for _array_element_3 _values
    (Utilities|Set|Add (Variables|Default|GetPlayset%20Actors%20Set) (Utilities|GetClass _array_element_3))
    (Development|PrintString (Utilities|String|ToString(Float) (* (* (* (.x (* (.BoxExtent (Collision|GetComponentBounds (Variables|/Script/Engine.Actor|GetRootComponent _array_element_3))) 0.01)) (.y (* (.BoxExtent (Collision|GetComponentBounds (Variables|/Script/Engine.Actor|GetRootComponent _array_element_3))) 0.01))) (.z (* (.BoxExtent (Collision|GetComponentBounds (Variables|/Script/Engine.Actor|GetRootComponent _array_element_3))) 0.01))) 8.0)))
    (Utilities|Array|Add (Variables|PlaysetGridLayout|GetObjectStatsArray) (Utilities|Struct|MakePlaysetGridLayoutStruc _array_element_3 (.BoxExtent (Collision|GetComponentBounds (Variables|/Script/Engine.Actor|GetRootComponent _array_element_3))) (* (* (* (.x (* (.BoxExtent (Collision|GetComponentBounds (Variables|/Script/Engine.Actor|GetRootComponent _array_element_3))) 0.01)) (.y (* (.BoxExtent (Collision|GetComponentBounds (Variables|/Script/Engine.Actor|GetRootComponent _array_element_3))) 0.01))) (.z (* (.BoxExtent (Collision|GetComponentBounds (Variables|/Script/Engine.Actor|GetRootComponent _array_element_3))) 0.01))) 8.0))))
  (bind _sortedarray (PlaysetGridLayout|StructureBubbleSort _objectstatsarray))
  (Variables|PlaysetGridLayout|SetObjectStatsArray _sortedarray)
  (Development|PrintString (Utilities|String|Append (Utilities|String|ToString(Integer) (Utilities|Array|Length _objectstatsarray)) " Objects gathered and sorted by bouding volume")))

(event Custom|Place%20Floor
  (bind _floor_object_bp (Variables|Default|GetFloor%20Object%20BP))
  (bind _staticmeshcomponent (Variables|/Script/Engine.StaticMeshActor|GetStaticMeshComponent _floor_object_bp))
  (Transformation|SetWorldLocationAndRotation _staticmeshcomponent "512.000000,-512.000000,0.000000" "0,180.000000,0.000000" false false)
  (Transformation|SetRelativeScale3D _staticmeshcomponent (Math|Vector|MakeVector (+ (/ (Math|Float|float+float (Utilities|Array|Get(acopy) (Variables|Default|GetGridSpaceOffsetsX) (- (Utilities|Array|Length (Variables|Default|GetGridSpaceOffsetsX)) 1)) (Utilities|Array|Get(acopy) (Variables|Default|GetGridSpacesX) (- (Utilities|Array|Length (Variables|Default|GetGridSpacesX)) 1)) 512.0) 512.0) 1.0) (+ (/ (Math|Float|float+float (Utilities|Array|Get(acopy) (Variables|Default|GetGridSpacesY) (- (Utilities|Array|Length (Variables|Default|GetGridSpacesY)) 1)) (Utilities|Array|Get(acopy) (Variables|Default|GetGridSpaceOffsetsY) (- (Utilities|Array|Length (Variables|Default|GetGridSpaceOffsetsY)) 1)) 512.0) 512.0) 1.0) 3.0)))

(event Custom|Swap%20Object%20Positions
  (bind _target_actor (Variables|Default|GetTarget_Actor))
  (bind _source_actor (Variables|Default|GetSource_Actor))
  (bind _rootcomponent (Variables|/Script/Engine.Actor|GetRootComponent _target_actor))
  (bind _rootcomponent_2 (Variables|/Script/Engine.Actor|GetRootComponent _source_actor))
  (Variables|Default|SetUtil_Vector (Transformation|GetWorldLocation _rootcomponent))
  (Transformation|SetWorldLocation _rootcomponent (Transformation|GetWorldLocation _rootcomponent_2) false false)
  (Transformation|SetWorldLocation _rootcomponent_2 (Variables|Default|GetUtil_Vector) false false))

(event Custom|Rot%20z180
  (for _array_element (Variables|PlaysetGridLayout|GetObjectStatsArray)
    (Transformation|AddActorLocalRotation (Utilities|Struct|BreakPlaysetGridLayoutStruc _array_element) "0,180.000000, 0" false false)
    (Transformation|SetActorLocation (Utilities|Struct|BreakPlaysetGridLayoutStruc _array_element) (+ (Transformation|GetWorldLocation (Variables|/Script/Engine.Actor|GetRootComponent (Utilities|Struct|BreakPlaysetGridLayoutStruc _array_element))) (* (Math|Vector|MakeVector (.x (- (Transformation|GetWorldLocation (Variables|/Script/Engine.Actor|GetRootComponent (Utilities|Struct|BreakPlaysetGridLayoutStruc _array_element))) (Math|Vector|MakeVector (.x (Collision|GetComponentBounds (Variables|/Script/Engine.Actor|GetRootComponent (Utilities|Struct|BreakPlaysetGridLayoutStruc _array_element)))) (.y (Collision|GetComponentBounds (Variables|/Script/Engine.Actor|GetRootComponent (Utilities|Struct|BreakPlaysetGridLayoutStruc _array_element))))))) (.y (- (Transformation|GetWorldLocation (Variables|/Script/Engine.Actor|GetRootComponent (Utilities|Struct|BreakPlaysetGridLayoutStruc _array_element))) (Math|Vector|MakeVector (.x (Collision|GetComponentBounds (Variables|/Script/Engine.Actor|GetRootComponent (Utilities|Struct|BreakPlaysetGridLayoutStruc _array_element)))) (.y (Collision|GetComponentBounds (Variables|/Script/Engine.Actor|GetRootComponent (Utilities|Struct|BreakPlaysetGridLayoutStruc _array_element)))))))) 2.0)) false false)))

(fn ArrayInvisible
  (params (Target%20Actor%20Array :type Array%20of%20Actor%20Object%20References))
  (for _array_element Target%20Actor%20Array
    (Rendering|SetActorHiddenInGame _array_element true)))

(fn ClearArrays
  (Utilities|Array|Clear (Variables|PlaysetGridLayout|GetObjectStatsArray))
  (Utilities|Map|Clear (Variables|PlaysetGridLayout|GetMaterialDictionary)))

(fn GatherObjects
  (CallFunction|GatherUniqueObjects))

(fn PlaceObjectsinaGrid
  (bind _objectstatsarray (Variables|PlaysetGridLayout|GetObjectStatsArray))
  (bind _yshelfmax (LocalVariables|GetYShelfMax))
  (bind _yshelfmax_2 (LocalVariables|GetYShelfMax))
  (bind _yshelfmax_3 (LocalVariables|GetYShelfMax))
  (bind _gridspacesx (Variables|Default|GetGridSpacesX))
  (bind _gridspacesy (Variables|Default|GetGridSpacesY))
  (if (Variables|PlaysetGridLayout|GetAuto%20sqaure%20grid)
    (LocalVariables|SetYShelfMax (Math|Float|Ceil (Math|Float|Sqrt (Math|Conversions|ToFloat(Integer) (Utilities|Array|Length (Variables|PlaysetGridLayout|GetObjectStatsArray))))))
    (else
      (LocalVariables|SetYShelfMax (Variables|PlaysetGridLayout|GetY%20Row%20Max%20Items))))
  (if (Variables|PlaysetGridLayout|GetAuto%20spacing)
    (LocalVariables|SetCurrent%20Column 0)
    (LocalVariables|SetCurrent%20Column%20Max%20XBounds 2e+02)
    (Utilities|Array|Clear _gridspacesx)
    (Utilities|Array|Clear (Variables|Default|GetGridSpaceOffsetsX))
    (Utilities|Array|Resize _gridspacesx (+ (/ (Utilities|Array|Length (Variables|PlaysetGridLayout|GetObjectStatsArray)) _yshelfmax) 1))
    (Utilities|Array|Clear _gridspacesy)
    (Utilities|Array|Clear (Variables|Default|GetGridSpaceOffsetsY))
    (Utilities|Array|Resize _gridspacesy _yshelfmax)
    (for _index (range (Utilities|Array|Length (Variables|PlaysetGridLayout|GetObjectStatsArray)))
      (if (> (.x (.BoundExtents_5_8F4FC81E43DE744EE77F1090E7247E6D (Utilities|Struct|BreakPlaysetGridLayoutStruc (Utilities|Array|Get(acopy) _objectstatsarray _index)))) (Utilities|Array|Get(acopy) (Variables|Default|GetGridSpacesX) (/ _index _yshelfmax_2)))
        (Utilities|Array|SetArrayElem (Variables|Default|GetGridSpacesX) (/ _index _yshelfmax_2) (.x (.BoundExtents_5_8F4FC81E43DE744EE77F1090E7247E6D (Utilities|Struct|BreakPlaysetGridLayoutStruc (Utilities|Array|Get(acopy) _objectstatsarray _index))))))
      (if (> (.y (.BoundExtents_5_8F4FC81E43DE744EE77F1090E7247E6D (Utilities|Struct|BreakPlaysetGridLayoutStruc (Utilities|Array|Get(acopy) _objectstatsarray _index)))) (Utilities|Array|Get(acopy) (Variables|Default|GetGridSpacesY) (Math|Integer|%25%28Integer%29 _index _yshelfmax_2)))
        (Utilities|Array|SetArrayElem (Variables|Default|GetGridSpacesY) (Math|Integer|%25%28Integer%29 _index _yshelfmax_2) (.y (.BoundExtents_5_8F4FC81E43DE744EE77F1090E7247E6D (Utilities|Struct|BreakPlaysetGridLayoutStruc (Utilities|Array|Get(acopy) _objectstatsarray _index)))))))
    (Utilities|Array|Add (Variables|Default|GetGridSpaceOffsetsX))
    (for _index_2 (range 1 (Utilities|Array|Length (Variables|Default|GetGridSpacesX)))
      (Utilities|Array|Add (Variables|Default|GetGridSpaceOffsetsX) (+ (+ (+ (Utilities|Array|Get(acopy) (Variables|Default|GetGridSpacesX) _index_2) (Utilities|Array|Get(acopy) (Variables|Default|GetGridSpacesX) (- _index_2 1))) (Variables|PlaysetGridLayout|GetAuto%20Grid%20border)) (Utilities|Array|Get(aref) (Variables|Default|GetGridSpaceOffsetsX) (- _index_2 1)))))
    (Utilities|Array|Add (Variables|Default|GetGridSpaceOffsetsY))
    (for _index_3 (range 1 (Utilities|Array|Length (Variables|Default|GetGridSpacesY)))
      (Utilities|Array|Add (Variables|Default|GetGridSpaceOffsetsY) (+ (+ (+ (Utilities|Array|Get(acopy) (Variables|Default|GetGridSpacesY) _index_3) (Utilities|Array|Get(acopy) (Variables|Default|GetGridSpacesY) (- _index_3 1))) (Variables|PlaysetGridLayout|GetAuto%20Grid%20border)) (Utilities|Array|Get(aref) (Variables|Default|GetGridSpaceOffsetsY) (- _index_3 1))))))
  (for (_array_element _array_index) (Variables|PlaysetGridLayout|GetObjectStatsArray)
    (if (and (.Prop_13_1577E0E74F6153A56091AC98240195CD (Utilities|Struct|BreakPlaysetGridLayoutStruc _array_element)) (Variables|PlaysetGridLayout|GetSplit%20Prop))
      (LocalVariables|SetGrid%20Offset%20local%20 (Variables|PlaysetGridLayout|GetSplit%20Offset))
      (else
        (LocalVariables|SetGrid%20Offset%20local%20 (Variables|PlaysetGridLayout|GetGrid%20Offset))))
    (if (Variables|PlaysetGridLayout|GetAuto%20spacing)
      (Transformation|SetActorLocation (Utilities|Struct|BreakPlaysetGridLayoutStruc _array_element) (+ (* (+ (Math|Vector|MakeVector (Utilities|Array|Get(aref) (Variables|Default|GetGridSpaceOffsetsX) (/ _array_index _yshelfmax_3)) (Utilities|Array|Get(aref) (Variables|Default|GetGridSpaceOffsetsY) (Math|Integer|%25%28Integer%29 _array_index _yshelfmax_3))) (LocalVariables|GetGrid%20Offset%20local%20)) (Variables|PlaysetGridLayout|GetGrid%20Transform)) (+ (- (Transformation|GetWorldLocation (Variables|/Script/Engine.Actor|GetRootComponent (Utilities|Struct|BreakPlaysetGridLayoutStruc _array_element))) (Math|Vector|MakeVector (.x (Collision|GetComponentBounds (Variables|/Script/Engine.Actor|GetRootComponent (Utilities|Struct|BreakPlaysetGridLayoutStruc _array_element)))) (.y (Collision|GetComponentBounds (Variables|/Script/Engine.Actor|GetRootComponent (Utilities|Struct|BreakPlaysetGridLayoutStruc _array_element)))))) (Math|Vector|MakeVector 0.0 0.0 (- (.z (.BoxExtent (Collision|GetComponentBounds (Variables|/Script/Engine.Actor|GetRootComponent (Utilities|Struct|BreakPlaysetGridLayoutStruc _array_element))))) (.z (Collision|GetComponentBounds (Variables|/Script/Engine.Actor|GetRootComponent (Utilities|Struct|BreakPlaysetGridLayoutStruc _array_element)))))))) false false)
      (Transformation|SetActorRotation (Utilities|Struct|BreakPlaysetGridLayoutStruc _array_element) "0, 0, 0" false)
      (else
        (Transformation|SetActorLocation (Utilities|Struct|BreakPlaysetGridLayoutStruc _array_element) (+ (* (+ (Math|Vector|MakeVector (Math|Conversions|ToFloat(Integer) (* (/ _array_index _yshelfmax_3) (Variables|PlaysetGridLayout|GetX%20Grid%20Space))) (Math|Conversions|ToFloat(Integer) (* (Math|Integer|%25%28Integer%29 _array_index _yshelfmax_3) (Variables|PlaysetGridLayout|GetY%20Grid%20Space)))) (LocalVariables|GetGrid%20Offset%20local%20)) (Variables|PlaysetGridLayout|GetGrid%20Transform)) (+ (- (Transformation|GetWorldLocation (Variables|/Script/Engine.Actor|GetRootComponent (Utilities|Struct|BreakPlaysetGridLayoutStruc _array_element))) (Math|Vector|MakeVector (.x (Collision|GetComponentBounds (Variables|/Script/Engine.Actor|GetRootComponent (Utilities|Struct|BreakPlaysetGridLayoutStruc _array_element)))) (.y (Collision|GetComponentBounds (Variables|/Script/Engine.Actor|GetRootComponent (Utilities|Struct|BreakPlaysetGridLayoutStruc _array_element)))))) (Math|Vector|MakeVector 0.0 0.0 (- (.z (.BoxExtent (Collision|GetComponentBounds (Variables|/Script/Engine.Actor|GetRootComponent (Utilities|Struct|BreakPlaysetGridLayoutStruc _array_element))))) (.z (Collision|GetComponentBounds (Variables|/Script/Engine.Actor|GetRootComponent (Utilities|Struct|BreakPlaysetGridLayoutStruc _array_element)))))))) false false)
        (Transformation|SetActorRotation (Utilities|Struct|BreakPlaysetGridLayoutStruc _array_element) "0, 0, 0" false)))))

(fn PopulateVolumeStats
  (for _array_element (Variables|PlaysetGridLayout|GetObjectStatsArray)
    (Utilities|Array|SetArrayElem (Variables|PlaysetGridLayout|GetObjectStatsArray) 0 (Utilities|Struct|MakePlaysetGridLayoutStruc (Utilities|Struct|BreakPlaysetGridLayoutStruc _array_element) (.BoxExtent (Collision|GetComponentBounds (Variables|/Script/Engine.Actor|GetRootComponent (Utilities|Struct|BreakPlaysetGridLayoutStruc _array_element)))) (* (* (* (.x (* (.BoxExtent (Collision|GetComponentBounds (Variables|/Script/Engine.Actor|GetRootComponent (Utilities|Struct|BreakPlaysetGridLayoutStruc _array_element)))) 0.01)) (.y (* (.BoxExtent (Collision|GetComponentBounds (Variables|/Script/Engine.Actor|GetRootComponent (Utilities|Struct|BreakPlaysetGridLayoutStruc _array_element)))) 0.01))) (.z (* (.BoxExtent (Collision|GetComponentBounds (Variables|/Script/Engine.Actor|GetRootComponent (Utilities|Struct|BreakPlaysetGridLayoutStruc _array_element)))) 0.01))) 8.0) (.Prop_13_1577E0E74F6153A56091AC98240195CD (Utilities|Struct|BreakPlaysetGridLayoutStruc _array_element))))))

(fn RefreshObjectStats
  (Utility|PopulateVolumeStats)
  (PlaysetGridLayout|SortByVolume))

(fn SortbyVolume
  (Utility|PopulateVolumeStats)
  (bind _sortedarray (PlaysetGridLayout|StructureBubbleSort (Variables|PlaysetGridLayout|GetObjectStatsArray)))
  (Variables|PlaysetGridLayout|SetObjectStatsArray _sortedarray))

(fn StructureBubbleSort
  (params (PlaysetLayoutArray :type Array%20of%20Playset%20Grid%20Layout%20Struc%20Structures))
  (for _index (range (+ (- (Utilities|Array|Length PlaysetLayoutArray) 2) 1))
    (if (> (.x (.BoundExtents_5_8F4FC81E43DE744EE77F1090E7247E6D (Utilities|Struct|BreakPlaysetGridLayoutStruc (Utilities|Array|Get(aref) PlaysetLayoutArray _index)))) (.x (.BoundExtents_5_8F4FC81E43DE744EE77F1090E7247E6D (Utilities|Struct|BreakPlaysetGridLayoutStruc (Utilities|Array|Get(aref) PlaysetLayoutArray (+ _index 1))))))
      (Utilities|Array|SwapArrayElements PlaysetLayoutArray _index (+ _index 1))
      (bind _output_get (LocalVariables|SetResort true))))
  (if (stale _output_get)
    (LocalVariables|SetResort false))
  (return PlaysetLayoutArray))

(fn ConstructionScript)
