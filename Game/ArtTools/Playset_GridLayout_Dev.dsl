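; Dev art tool that gathers unique BuildingSMActors in the level and lays them out in a grid for playsets.
; Gather Unique Objects dedupes by building type plus texture names, and other props by class, recording extents and volume.
; StructureBubbleSort makes one visible swap pass on X extent, despite the "sorted by volume" labels.
; PlaceObjectsinaGrid supports auto-square rows, auto column spacing and prop splitting, correcting for pivot-to-bounds offsets.
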(event EventBeginPlay)

(event Collision|EventActorBeginOverlap
  (params (OtherActor :type Actor%20Object%20Reference)))

(event EventTick
  (params (DeltaSeconds :type Float%20%28single-precision%29)))

(event Custom|WIP%20Actor%20Details
  (bind _target_actor (Variables|Default|GetTarget_Actor))
  (bind _rootcomponent (Variables|/Script/Engine.Actor|GetRootComponent _target_actor))
  (bind (_origin _boxextent) (Collision|GetComponentBounds _rootcomponent))
  (Variables|Default|SetTarget%20Actor%20Origin _origin)
  (Variables|Default|SetTarget%20Actor%20Bounds _boxextent)
  (Variables|Default|SetTarget%20Actor%20Position (Transformation|GetWorldLocation _rootcomponent)))

(event Custom|Gather%20Unique%20Objects
  (bind _objectstatsarray (Variables|PlaysetGridLayout|GetObjectStatsArray))
  (Utilities|Array|Clear (Variables|PlaysetGridLayout|GetObjectStatsArray))
  (Utilities|Array|Clear (Variables|Default|GetActorArray))
  (Utilities|Map|Clear (Variables|PlaysetGridLayout|GetMaterialDictionary))
  (bind _outactors (Actor|GetAllActorsOfClass "/Script/FortniteGame.BuildingSMActor"))
  (for _array_element _outactors
    (Variables|Default|SetObjectClassName "")
    (Utilities|Casting|CastToBuildingWall _array_element
      (:then
        (Variables|Default|SetObjectClassName "BuildingWall"))
      (:CastFailed
        (Utilities|Casting|CastToBuildingFloor _array_element
          (:then
            (Variables|Default|SetObjectClassName "BuildingFloor"))
          (:CastFailed
            (Utilities|Casting|CastToBuildingStairs _array_element
              (:then
                (Variables|Default|SetObjectClassName "BuildingStairs"))
              (:CastFailed
                (Utilities|Casting|CastToBuildingRoof _array_element
                  (:then
                    (Variables|Default|SetObjectClassName "BuildingRoof"))
                  (:CastFailed
                    (if (not (Utilities|Set|ContainsItem (Variables|Default|GetPlayset%20Actors%20Set) (Utilities|GetClass _array_element)))
                      (Utilities|Set|Add (Variables|Default|GetPlayset%20Actors%20Set) (Utilities|GetClass _array_element))
                      (Development|PrintString (Utilities|String|ToString(Float) (* (* (* (.x (* (.BoxExtent (Collision|GetComponentBounds (Variables|/Script/Engine.Actor|GetRootComponent _array_element))) 0.01)) (.y (* (.BoxExtent (Collision|GetComponentBounds (Variables|/Script/Engine.Actor|GetRootComponent _array_element))) 0.01))) (.z (* (.BoxExtent (Collision|GetComponentBounds (Variables|/Script/Engine.Actor|GetRootComponent _array_element))) 0.01))) 8.0)))
                      (Utilities|Array|Add (Variables|PlaysetGridLayout|GetObjectStatsArray) (Utilities|Struct|MakePlaysetGridLayoutStruc _array_element (.BoxExtent (Collision|GetComponentBounds (Variables|/Script/Engine.Actor|GetRootComponent _array_element))) (* (* (* (.x (* (.BoxExtent (Collision|GetComponentBounds (Variables|/Script/Engine.Actor|GetRootComponent _array_element))) 0.01)) (.y (* (.BoxExtent (Collision|GetComponentBounds (Variables|/Script/Engine.Actor|GetRootComponent _array_element))) 0.01))) (.z (* (.BoxExtent (Collision|GetComponentBounds (Variables|/Script/Engine.Actor|GetRootComponent _array_element))) 0.01))) 8.0) true)))))))))))
    (Variables|Default|SetObjectMatName "")
    (bind _outtexturedataarray (Building|GetBuildingTextureData _array_element :!class /Script/FortniteGame.BuildingSMActor))
    (for _array_element_2 _outtexturedataarray
      (Development|PrintString "Hello")
      (if (> (Utilities|String|Len (Variables|Default|GetObjectMatName)) 0)
        (Variables|Default|SetObjectMatName (Utilities|String|Append (Variables|Default|GetObjectMatName) ":" (Utilities|GetDisplayName _array_element_2) ""))
        (else
          (Variables|Default|SetObjectMatName (Utilities|GetDisplayName _array_element_2)))))
    (Utilities|Map|Add (Variables|PlaysetGridLayout|GetMaterialDictionary) (Utilities|String|Append (Variables|Default|GetObjectClassName) "-" (Variables|Default|GetObjectMatName)) _array_element))
  (bind _values (Utilities|Map|Values (Variables|PlaysetGridLayout|GetMaterialDictionary)))
  (for _array_element_3 _values
    (Utilities|Set|Add (Variables|Default|GetPlayset%20Actors%20Set) (Utilities|GetClass _array_element_3))
    (Development|PrintString (Utilities|String|ToString(Float) (* (* (* (.x (* (.BoxExtent (Collision|GetComponentBounds (Variables|/Script/Engine.Actor|GetRootComponent _array_element_3))) 0.01)) (.y (* (.BoxExtent (Collision|GetComponentBounds (Variables|/Script/Engine.Actor|GetRootComponent _array_element_3))) 0.01))) (.z (* (.BoxExtent (Collision|GetComponentBounds (Variables|/Script/Engine.Actor|GetRootComponent _array_element_3))) 0.01))) 8.0)))
    (Utilities|Array|Add (Variables|PlaysetGridLayout|GetObjectStatsArray) (Utilities|Struct|MakePlaysetGridLayoutStruc _array_element_3 (.BoxExtent (Collision|GetComponentBounds (Variables|/Script/Engine.Actor|GetRootComponent _array_element_3))) (* (* (* (.x (* (.BoxExtent (Collision|GetComponentBounds (Variables|/Script/Engine.Actor|GetRootComponent _array_element_3))) 0.01)) (.y (* (.BoxExtent (Collision|GetComponentBounds (Variables|/Script/Engine.Actor|GetRootComponent _array_element_3))) 0.01))) (.z (* (.BoxExtent (Collision|GetComponentBounds (Variables|/Script/Engine.Actor|GetRootComponent _array_element_3))) 0.01))) 8.0))))
  (bind _sortedarray (PlaysetGridLayout|StructureBubbleSort _objectstatsarray))
  (Variables|PlaysetGridLayout|SetObjectStatsArray _sortedarray)
  (Development|PrintString (Utilities|String|Append (Utilities|String|ToString(Integer) (Utilities|Array|Length _objectstatsarray)) " Objects gathered and sorted by bouding volume")))

(fn ArrayInvisible
  (for _array_element (Variables|Default|GetActorArray)
    (Rendering|SetActorHiddenInGame _array_element true)))

(fn OLD_GatherUniqueObjects
  (bind _objectstatsarray (Variables|PlaysetGridLayout|GetObjectStatsArray))
  (Utilities|Array|Clear (Variables|PlaysetGridLayout|GetObjectStatsArray))
  (Utilities|Array|Clear (Variables|Default|GetActorArray))
  (bind _outactors (Actor|GetAllActorsOfClass "/Script/FortniteGame.BuildingSMActor"))
  (for _array_element _outactors
    (if (not (Utilities|Set|ContainsItem (Variables|Default|GetPlayset%20Actors%20Set) (Utilities|GetClass _array_element)))
      (Utilities|Set|Add (Variables|Default|GetPlayset%20Actors%20Set) (Utilities|GetClass _array_element))
      (Development|PrintString (Utilities|String|ToString(Float) (* (* (* (.x (* (.BoxExtent (Collision|GetComponentBounds (Variables|/Script/Engine.Actor|GetRootComponent _array_element))) 0.01)) (.y (* (.BoxExtent (Collision|GetComponentBounds (Variables|/Script/Engine.Actor|GetRootComponent _array_element))) 0.01))) (.z (* (.BoxExtent (Collision|GetComponentBounds (Variables|/Script/Engine.Actor|GetRootComponent _array_element))) 0.01))) 8.0)))
      (Utilities|Array|Add (Variables|PlaysetGridLayout|GetObjectStatsArray) (Utilities|Struct|MakePlaysetGridLayoutStruc _array_element (.BoxExtent (Collision|GetComponentBounds (Variables|/Script/Engine.Actor|GetRootComponent _array_element))) (* (* (* (.x (* (.BoxExtent (Collision|GetComponentBounds (Variables|/Script/Engine.Actor|GetRootComponent _array_element))) 0.01)) (.y (* (.BoxExtent (Collision|GetComponentBounds (Variables|/Script/Engine.Actor|GetRootComponent _array_element))) 0.01))) (.z (* (.BoxExtent (Collision|GetComponentBounds (Variables|/Script/Engine.Actor|GetRootComponent _array_element))) 0.01))) 8.0)))))
  (bind _sortedarray (PlaysetGridLayout|StructureBubbleSort _objectstatsarray))
  (Variables|PlaysetGridLayout|SetObjectStatsArray _sortedarray)
  (Development|PrintString (Utilities|String|Append (Utilities|String|ToString(Integer) (Utilities|Array|Length _objectstatsarray)) " Objects gathered and sorted by bouding volume")))

(fn PlaceObjectsinaGrid
  (bind _objectstatsarray (Variables|PlaysetGridLayout|GetObjectStatsArray))
  (bind _y_grid_space (Variables|PlaysetGridLayout|GetY%20Grid%20Space))
  (bind _yshelfmax (LocalVariables|GetYShelfMax))
  (bind _yshelfmax_2 (LocalVariables|GetYShelfMax))
  (bind _gridspacesx (Variables|Default|GetGridSpacesX))
  (if (Variables|PlaysetGridLayout|GetAuto%20sqaure%20grid)
    (LocalVariables|SetYShelfMax (Math|Float|Ceil (Math|Float|Sqrt (Math|Conversions|ToFloat(Integer) (Utilities|Array|Length (Variables|PlaysetGridLayout|GetObjectStatsArray))))))
    (else
      (LocalVariables|SetYShelfMax (Variables|PlaysetGridLayout|GetY%20Row%20Max%20Items))))
  (if (Variables|PlaysetGridLayout|GetAuto%20spacing)
    (LocalVariables|SetCurrent%20Column 0)
    (LocalVariables|SetCurrent%20Column%20Max%20XBounds 2e+02)
    (Utilities|Array|Clear _gridspacesx)
    (Utilities|Array|Clear (Variables|Default|GetGridSpaceOffsetsX))
    (Utilities|Array|Resize _gridspacesx (/ (Utilities|Array|Length (Variables|PlaysetGridLayout|GetObjectStatsArray)) (LocalVariables|GetYShelfMax)))
    (for _index (range (Utilities|Array|Length (Variables|PlaysetGridLayout|GetObjectStatsArray)))
      (if (> (.x (.BoundExtents_5_8F4FC81E43DE744EE77F1090E7247E6D (Utilities|Struct|BreakPlaysetGridLayoutStruc (Utilities|Array|Get(acopy) _objectstatsarray _index)))) (Utilities|Array|Get(acopy) (Variables|Default|GetGridSpacesX) (/ _index _yshelfmax)))
        (Utilities|Array|Insert (Variables|Default|GetGridSpacesX) (.x (.BoundExtents_5_8F4FC81E43DE744EE77F1090E7247E6D (Utilities|Struct|BreakPlaysetGridLayoutStruc (Utilities|Array|Get(acopy) _objectstatsarray _index)))) (/ _index _yshelfmax))))
    (Utilities|Array|Add (Variables|Default|GetGridSpaceOffsetsX))
    (for _index_2 (range 1 (Utilities|Array|Length (Variables|Default|GetGridSpacesX)))
      (Utilities|Array|Add (Variables|Default|GetGridSpaceOffsetsX) (+ (+ (+ (Utilities|Array|Get(acopy) (Variables|Default|GetGridSpacesX) _index_2) (Utilities|Array|Get(acopy) (Variables|Default|GetGridSpacesX) (- _index_2 1))) (Variables|PlaysetGridLayout|GetAuto%20Grid%20border)) (Utilities|Array|Get(aref) (Variables|Default|GetGridSpaceOffsetsX) (- _index_2 1))))))
  (for (_array_element _array_index) (Variables|PlaysetGridLayout|GetObjectStatsArray)
    (if (Variables|PlaysetGridLayout|GetAuto%20spacing)
      (Transformation|SetActorLocation (Utilities|Struct|BreakPlaysetGridLayoutStruc _array_element) (+ (* (Math|Vector|MakeVector (Utilities|Array|Get(aref) (Variables|Default|GetGridSpaceOffsetsX) (/ _array_index _yshelfmax_2)) (Math|Conversions|ToFloat(Integer) (* (Math|Integer|%25%28Integer%29 _array_index _yshelfmax_2) _y_grid_space))) (Variables|PlaysetGridLayout|GetGrid%20Transform)) (+ (- (Transformation|GetWorldLocation (Variables|/Script/Engine.Actor|GetRootComponent (Utilities|Struct|BreakPlaysetGridLayoutStruc _array_element))) (Math|Vector|MakeVector (.x (Collision|GetComponentBounds (Variables|/Script/Engine.Actor|GetRootComponent (Utilities|Struct|BreakPlaysetGridLayoutStruc _array_element)))) (.y (Collision|GetComponentBounds (Variables|/Script/Engine.Actor|GetRootComponent (Utilities|Struct|BreakPlaysetGridLayoutStruc _array_element)))))) (Math|Vector|MakeVector 0.0 0.0 (- (.z (.BoxExtent (Collision|GetComponentBounds (Variables|/Script/Engine.Actor|GetRootComponent (Utilities|Struct|BreakPlaysetGridLayoutStruc _array_element))))) (.z (Collision|GetComponentBounds (Variables|/Script/Engine.Actor|GetRootComponent (Utilities|Struct|BreakPlaysetGridLayoutStruc _array_element)))))))) false false)
      (Transformation|SetActorRotation (Utilities|Struct|BreakPlaysetGridLayoutStruc _array_element) "0, 0, 0" false)
      (elif (and (.Prop_13_1577E0E74F6153A56091AC98240195CD (Utilities|Struct|BreakPlaysetGridLayoutStruc _array_element)) (Variables|PlaysetGridLayout|GetSplit%20Prop))
        (LocalVariables|SetGrid%20Offset%20local%20 (Variables|PlaysetGridLayout|GetSplit%20Offset))
        (Transformation|SetActorLocation (Utilities|Struct|BreakPlaysetGridLayoutStruc _array_element) (+ (* (+ (Math|Vector|MakeVector (Math|Conversions|ToFloat(Integer) (* (/ _array_index _yshelfmax_2) (Variables|PlaysetGridLayout|GetX%20Grid%20Space))) (Math|Conversions|ToFloat(Integer) (* (Math|Integer|%25%28Integer%29 _array_index _yshelfmax_2) _y_grid_space))) (Variables|PlaysetGridLayout|GetGrid%20Offset)) (Variables|PlaysetGridLayout|GetGrid%20Transform)) (+ (- (Transformation|GetWorldLocation (Variables|/Script/Engine.Actor|GetRootComponent (Utilities|Struct|BreakPlaysetGridLayoutStruc _array_element))) (Math|Vector|MakeVector (.x (Collision|GetComponentBounds (Variables|/Script/Engine.Actor|GetRootComponent (Utilities|Struct|BreakPlaysetGridLayoutStruc _array_element)))) (.y (Collision|GetComponentBounds (Variables|/Script/Engine.Actor|GetRootComponent (Utilities|Struct|BreakPlaysetGridLayoutStruc _array_element)))))) (Math|Vector|MakeVector 0.0 0.0 (- (.z (.BoxExtent (Collision|GetComponentBounds (Variables|/Script/Engine.Actor|GetRootComponent (Utilities|Struct|BreakPlaysetGridLayoutStruc _array_element))))) (.z (Collision|GetComponentBounds (Variables|/Script/Engine.Actor|GetRootComponent (Utilities|Struct|BreakPlaysetGridLayoutStruc _array_element)))))))) false false)
        (Transformation|SetActorRotation (Utilities|Struct|BreakPlaysetGridLayoutStruc _array_element) "0, 0, 0" false)
        (else
          (LocalVariables|SetGrid%20Offset%20local%20 (Variables|PlaysetGridLayout|GetGrid%20Offset)))))))

(fn PopulateVolumeStats
  (for _array_element (Variables|PlaysetGridLayout|GetObjectStatsArray)
    (Utilities|Array|SetArrayElem (Variables|PlaysetGridLayout|GetObjectStatsArray) 0 (Utilities|Struct|MakePlaysetGridLayoutStruc (Utilities|Struct|BreakPlaysetGridLayoutStruc _array_element) (.BoxExtent (Collision|GetComponentBounds (Variables|/Script/Engine.Actor|GetRootComponent (Utilities|Struct|BreakPlaysetGridLayoutStruc _array_element)))) (* (* (* (.x (* (.BoxExtent (Collision|GetComponentBounds (Variables|/Script/Engine.Actor|GetRootComponent (Utilities|Struct|BreakPlaysetGridLayoutStruc _array_element)))) 0.01)) (.y (* (.BoxExtent (Collision|GetComponentBounds (Variables|/Script/Engine.Actor|GetRootComponent (Utilities|Struct|BreakPlaysetGridLayoutStruc _array_element)))) 0.01))) (.z (* (.BoxExtent (Collision|GetComponentBounds (Variables|/Script/Engine.Actor|GetRootComponent (Utilities|Struct|BreakPlaysetGridLayoutStruc _array_element)))) 0.01))) 8.0) (.Prop_13_1577E0E74F6153A56091AC98240195CD (Utilities|Struct|BreakPlaysetGridLayoutStruc _array_element))))))

(fn SortbyVolume
  (CallFunction|PopulateVolumeStats)
  (bind _sortedarray (PlaysetGridLayout|StructureBubbleSort (Variables|PlaysetGridLayout|GetObjectStatsArray)))
  (Variables|PlaysetGridLayout|SetObjectStatsArray _sortedarray))

(fn StructureBubbleSort
  (params (PlaysetLayoutArray :type Array%20of%20Playset%20Grid%20Layout%20Struc%20Structures))
  (for _index (range (+ (- (Utilities|Array|Length PlaysetLayoutArray) 2) 1))
    (if (> (.x (.BoundExtents_5_8F4FC81E43DE744EE77F1090E7247E6D (Utilities|Struct|BreakPlaysetGridLayoutStruc (Utilities|Array|Get(aref) PlaysetLayoutArray _index)))) (.x (.BoundExtents_5_8F4FC81E43DE744EE77F1090E7247E6D (Utilities|Struct|BreakPlaysetGridLayoutStruc (Utilities|Array|Get(aref) PlaysetLayoutArray (+ _index 1))))))
      (Utilities|Array|SwapArrayElements PlaysetLayoutArray _index (+ _index 1))
      (bind _output_get (LocalVariables|SetResort true))))
  (if (stale _output_get)
    (LocalVariables|SetResort false))
  (return PlaysetLayoutArray))

(fn ConstructionScript)
