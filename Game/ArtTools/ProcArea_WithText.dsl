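; Procedural area marker actor that builds a rectangular box-line outline with an editable text label, judging by its ArtTools path an art-side tool.
; The construction script sets the TextRender text and size, then adds four ISM_LineBox instances scaled from MaxX, MaxY and line dimensions.
; The first edge is added in local space; the other three compose with the root's world transform and are added in world space.
; BeginPlay, overlap and Tick events are empty.
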(event EventBeginPlay)

(event Collision|EventActorBeginOverlap
  (params (OtherActor :type Actor%20Object%20Reference)))

(event EventTick
  (params (DeltaSeconds :type Float%20%28single-precision%29)))

(fn ConstructionScript
  (bind _textrender (Variables|Default|GetTextRender))
  (Rendering|Components|TextRender|SetWorldSize _textrender 3.2e+02)
  (Rendering|Components|TextRender|SetText _textrender (Utilities|Text|ToText(String) (Variables|Default|GetYour%20Text)))
  (for _index (range 4)
    (switch int _index
      (:0
        (Components|InstancedStaticMesh|AddInstance (Variables|Default|GetISM_LineBox) (Math|Transform|MakeTransform (Math|Vector|MakeVector 0.0 0.0 (* (Variables|Default|GetLineZ_dim) 0.5)) "0, 0, 0" (Math|Vector|MakeVector (Math|Conversions|ToFloat(Integer) (Variables|Default|GetLineX_dim)) (Math|Conversions|ToFloat(Integer) (- (Variables|Default|GetMaxY) (Variables|Default|GetLineX_dim))) (Math|Conversions|ToFloat(Integer) (Variables|Default|GetLineZ_dim)))) false))
      (:1
        (Components|InstancedStaticMesh|AddInstanceWorldSpace (Variables|Default|GetISM_LineBox) (Math|Transform|ComposeTransforms (Math|Transform|MakeTransform (Math|Vector|MakeVector (* (Variables|Default|GetMaxX) 0.5) (* (Variables|Default|GetMaxY) -0.5) (* (Variables|Default|GetLineZ_dim) 0.5)) "0,90.000000, 0" (Math|Vector|MakeVector (Math|Conversions|ToFloat(Integer) (Variables|Default|GetLineX_dim)) (Math|Conversions|ToFloat(Integer) (+ (Variables|Default|GetMaxX) (Variables|Default|GetLineX_dim))) (Math|Conversions|ToFloat(Integer) (Variables|Default|GetLineZ_dim)))) (Transformation|GetWorldTransform (Variables|Default|GetDefaultSceneRoot)))))
      (:2
        (Components|InstancedStaticMesh|AddInstanceWorldSpace (Variables|Default|GetISM_LineBox) (Math|Transform|ComposeTransforms (Math|Transform|MakeTransform (Math|Vector|MakeVector (Math|Conversions|ToFloat(Integer) (Variables|Default|GetMaxX)) 0.0 (* (Variables|Default|GetLineZ_dim) 0.5)) "0, 0, 0" (Math|Vector|MakeVector (Math|Conversions|ToFloat(Integer) (Variables|Default|GetLineX_dim)) (Math|Conversions|ToFloat(Integer) (- (Variables|Default|GetMaxY) (Variables|Default|GetLineX_dim))) (Math|Conversions|ToFloat(Integer) (Variables|Default|GetLineZ_dim)))) (Transformation|GetWorldTransform (Variables|Default|GetDefaultSceneRoot)))))
      (:3
        (Components|InstancedStaticMesh|AddInstanceWorldSpace (Variables|Default|GetISM_LineBox) (Math|Transform|ComposeTransforms (Math|Transform|MakeTransform (Math|Vector|MakeVector (* (Variables|Default|GetMaxX) 0.5) (* (Variables|Default|GetMaxY) 0.5) (* (Variables|Default|GetLineZ_dim) 0.5)) "0,90.000000, 0" (Math|Vector|MakeVector (Math|Conversions|ToFloat(Integer) (Variables|Default|GetLineX_dim)) (Math|Conversions|ToFloat(Integer) (+ (Variables|Default|GetMaxX) (Variables|Default|GetLineX_dim))) (Math|Conversions|ToFloat(Integer) (Variables|Default|GetLineZ_dim)))) (Transformation|GetWorldTransform (Variables|Default|GetDefaultSceneRoot))))))))
