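; Art tool Blueprint that captures a screenshot sequence across a time-of-day range in Fortnite.
; renderTOD validates start, end and frame count, freezes time-of-day speed, toggles the UI and starts a looping half-second timer.
; Each ScreenShotTOD tick sets the time, takes a console-command screenshot and advances by range over frame count.
; Past the end, it clears the timer with a zero interval and sets time to half the end value.
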(event Custom|errorChecking
  (bind _start_time (Variables|RenderTimeofDay|GetStart%20Time))
  ;[0.135 0.1128 0.1196] error checking
  (Utilities|FlowControl|Sequence
    (:then_0
      (if (not (< (Variables|RenderTimeofDay|GetStart%20Time) (Variables|RenderTimeofDay|GetEnd%20Time)))
        (Variables|RenderTimeofDay|SetEnd%20Time _start_time)
        (Variables|Default|SetinvertTOD _start_time)
        (Variables|RenderTimeofDay|SetStart%20Time (Variables|Default|GetinvertTOD))))
    (:then_1
      (if (not (>= (Variables|RenderTimeofDay|GetFrame%20Count) 1.0))
        (Variables|RenderTimeofDay|SetFrame%20Count 1)))
    (:then_2
      (if (!= (Variables|RenderTimeofDay|GetStart%20Time) (Variables|RenderTimeofDay|GetEnd%20Time))
        _
        (elif (== (Variables|RenderTimeofDay|GetEnd%20Time) 24)
          (Variables|RenderTimeofDay|SetStart%20Time (- (Variables|RenderTimeofDay|GetStart%20Time) 1.0))
          (else
            (Variables|RenderTimeofDay|SetEnd%20Time (+ (Variables|RenderTimeofDay|GetEnd%20Time) 1.0))))))))

(event Custom|ScreenShotTOD
  (bind _tod (Variables|Default|GetTOD))
  ;[0.122 0.1987 0.295] set time of day, screenshot, frame control, increment time of day and stop after 24
  (if (<= (Variables|Default|GetTOD) (Variables|RenderTimeofDay|GetEnd%20Time))
    (TimeOfDay|SetTimeOfDay (Variables|Default|GetTOD) :!class /Script/FortniteGame.FortKismetLibrary)
    (Development|ExecuteConsoleCommand "screenshot")
    (Development|PrintString (Utilities|String|ToString(Float) _tod))
    (Variables|Default|SetTOD (+ _tod (/ (- (Variables|RenderTimeofDay|GetEnd%20Time) (Variables|RenderTimeofDay|GetStart%20Time)) (Variables|RenderTimeofDay|GetFrame%20Count))))
    (else
      (Development|PrintString "-------------------- COMPLETE")
      (Utilities|Time|SetTimerbyFunctionName 0 "ScreenshotTOD")
      (TimeOfDay|SetTimeOfDay (/ (Variables|RenderTimeofDay|GetEnd%20Time) 2) :!class /Script/FortniteGame.FortKismetLibrary))))

(event Custom|renderTOD
  ;[1 0.587 0] render time of day, timer via function
  (CallFunction|ErrorChecking)
  (TimeOfDay|SetTimeOfDay (Variables|RenderTimeofDay|GetStart%20Time) :!class /Script/FortniteGame.FortKismetLibrary)
  (TimeOfDay|SetTimeOfDaySpeed 0.0 :!class /Script/FortniteGame.FortKismetLibrary)
  (Development|ExecuteConsoleCommand "ui.toggle")
  (Variables|Default|SetTOD (Variables|RenderTimeofDay|GetStart%20Time))
  (Utilities|Time|SetTimerbyFunctionName 0 "ScreenshotTOD" 0.5 true))

(fn ConstructionScript)
