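; RenderTreeBillboard is an art-tool Blueprint that poses a tree for rendering one billboard view at a time.
; Its construction script clamps BillboardToRender to the billboard count and converts it to an evenly spaced angle around 360 degrees.
; For side views it sets the tree's rotation to that angle, adds a -90 degree tilt and records CurrentAngle.
; With Top enabled it only moves the tree, using ZoffsetTop as height.
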(event EventBeginPlay)

(event Collision|EventActorBeginOverlap
  (params (OtherActor :type Actor%20Object%20Reference)))

(event EventTick
  (params (DeltaSeconds :type Float%20%28single-precision%29)))

(fn ConstructionScript
  (bind _number_of_billboards (Variables|Default|GetNumber%20of%20Billboards))
  (bind _returnvalue (Math|Conversions|ToFloat(Integer) _number_of_billboards))
  (bind _returnvalue_2 (- _number_of_billboards 1))
  (bind _billboardtorender (Variables|Default|GetBillboardToRender))
  (bind _returnvalue_3 (Math|Integer|Clamp(Integer) _billboardtorender 0 _returnvalue_2))
  (bind _returnvalue_4 (Math|Conversions|ToFloat(Integer) _returnvalue_3))
  (bind _returnvalue_5 (/ 360 _returnvalue))
  (bind _returnvalue_6 (* _returnvalue_5 _returnvalue_4))
  (bind _tree (Variables|Default|GetTree))
  (if (Variables|Default|GetTop)
    (Transformation|SetWorldLocation _tree (Math|Vector|MakeVector 0.0 100 (Variables|Default|GetZoffsetTop)) false false)
    (else
      (Transformation|SetWorldRotation _tree (Math|Rotator|MakeRotator 0.0 0.0 _returnvalue_6) false false)
      (Transformation|AddWorldRotation _tree (Math|Rotator|MakeRotator -90) false false)
      (Variables|Default|SetCurrentAngle(dontedit) _returnvalue_6))))
