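; Art-tool spline mesh actor that lays a mesh along an editable spline and can deform landscapes to match it.
; The construction script, duplicated in UpdateMeshfromSplineData, fits the base component to the first segment and adds mesh copies for the rest.
; Added segments optionally use tension-scaled auto-smoothed tangents, and all segments can take point Y/Z scale.
; A custom event applies the spline to every LandscapeProxy via EditorApplySpline.
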(event EventBeginPlay)

(event Collision|EventActorBeginOverlap
  (params (OtherActor :type Actor%20Object%20Reference)))

(event EventTick
  (params (DeltaSeconds :type Float%20%28single-precision%29)))

(event Custom|Conform%20Landscape%20to%20Spline
  (bind _outactors (Actor|GetAllActorsOfClass "/Script/Landscape.LandscapeProxy"))
  (for _array_element _outactors
    (Landscape|Editor|EditorApplySpline _array_element (Variables|Default|GetSpline) 5e+01 5e+01 5e+01 5e+01 0.0 0.0 20 true true 0 "")))

(fn UpdateMeshfromSplineData
  (bind _spline (Variables|Default|GetSpline))
  (bind (_location _tangent) (Spline|GetLocationandTangentatSplinePoint _spline 1 "Local"))
  (bind (_location_2 _tangent_2) (Spline|GetLocationandTangentatSplinePoint _spline 0 "Local"))
  (bind _splinemeshcomponent (Variables|SplineMeshActor|GetSplineMeshComponent))
  (bind _spline_2 (Variables|Default|GetSpline))
  (SplineMesh|SetStartandEnd _splinemeshcomponent _location_2 _tangent_2 _location _tangent true :!class /Script/Engine.SplineMeshComponent)
  (if (Variables|Default|GetUseScaling)
    (SplineMesh|SetStartScale _splinemeshcomponent (Math|Vector2D|MakeVector2D (.y (Spline|GetScaleatSplinePoint _spline 0)) (.z (Spline|GetScaleatSplinePoint _spline 0))) true :!class /Script/Engine.SplineMeshComponent)
    (SplineMesh|SetEndScale _splinemeshcomponent (Math|Vector2D|MakeVector2D (.y (Spline|GetScaleatSplinePoint _spline 1)) (.z (Spline|GetScaleatSplinePoint _spline 1))) true :!class /Script/Engine.SplineMeshComponent))
  (for _index (range 1 (+ (- (Spline|GetNumberOfSplinePoints _spline_2) 2) 1))
    (bind _returnvalue (AddComponent|NODE_AddSplineMeshComponent-1))
    (Components|StaticMesh|SetStaticMesh _returnvalue (Variables|/Script/Engine.StaticMeshComponent|GetStaticMesh (Variables|SplineMeshActor|GetSplineMeshComponent)))
    (SplineMesh|SetStartandEnd _returnvalue (Spline|GetLocationandTangentatSplinePoint _spline_2 _index "Local") (Math|Vector|SelectVector (* (Math|Vector|Normalize (.Tangent (Spline|GetLocationandTangentatSplinePoint _spline_2 _index "Local"))) (* (Math|Vector|VectorLength (- (Spline|GetLocationandTangentatSplinePoint _spline_2 (+ _index 1) "Local") (Spline|GetLocationandTangentatSplinePoint _spline_2 _index "Local"))) (Variables|Default|GetTension))) (.Tangent (Spline|GetLocationandTangentatSplinePoint _spline_2 _index "Local")) (Variables|Default|GetAutoSmoothTangent)) (Spline|GetLocationandTangentatSplinePoint _spline_2 (+ _index 1) "Local") (Math|Vector|SelectVector (* (Math|Vector|Normalize (.Tangent (Spline|GetLocationandTangentatSplinePoint _spline_2 (+ _index 1) "Local"))) (* (Math|Vector|VectorLength (- (Spline|GetLocationandTangentatSplinePoint _spline_2 (+ _index 1) "Local") (Spline|GetLocationandTangentatSplinePoint _spline_2 _index "Local"))) (Variables|Default|GetTension))) (.Tangent (Spline|GetLocationandTangentatSplinePoint _spline_2 (+ _index 1) "Local")) (Variables|Default|GetAutoSmoothTangent)) true :!class /Script/Engine.SplineMeshComponent)
    (if (Variables|Default|GetUseScaling)
      (SplineMesh|SetStartScale _returnvalue (Math|Vector2D|MakeVector2D (.y (Spline|GetScaleatSplinePoint _spline_2 _index)) (.z (Spline|GetScaleatSplinePoint _spline_2 _index))) true :!class /Script/Engine.SplineMeshComponent)
      (SplineMesh|SetEndScale _returnvalue (Math|Vector2D|MakeVector2D (.y (Spline|GetScaleatSplinePoint _spline_2 (+ _index 1))) (.z (Spline|GetScaleatSplinePoint _spline_2 (+ _index 1)))) true :!class /Script/Engine.SplineMeshComponent))))

(fn ConstructionScript
  (bind _spline (Variables|Default|GetSpline))
  (bind (_location _tangent) (Spline|GetLocationandTangentatSplinePoint _spline 1 "Local"))
  (bind (_location_2 _tangent_2) (Spline|GetLocationandTangentatSplinePoint _spline 0 "Local"))
  (bind _splinemeshcomponent (Variables|SplineMeshActor|GetSplineMeshComponent))
  (bind _spline_2 (Variables|Default|GetSpline))
  (SplineMesh|SetStartandEnd _splinemeshcomponent _location_2 _tangent_2 _location _tangent true :!class /Script/Engine.SplineMeshComponent)
  (if (Variables|Default|GetUseScaling)
    (SplineMesh|SetStartScale _splinemeshcomponent (Math|Vector2D|MakeVector2D (.y (Spline|GetScaleatSplinePoint _spline 0)) (.z (Spline|GetScaleatSplinePoint _spline 0))) true :!class /Script/Engine.SplineMeshComponent)
    (SplineMesh|SetEndScale _splinemeshcomponent (Math|Vector2D|MakeVector2D (.y (Spline|GetScaleatSplinePoint _spline 1)) (.z (Spline|GetScaleatSplinePoint _spline 1))) true :!class /Script/Engine.SplineMeshComponent))
  (for _index (range 1 (+ (- (Spline|GetNumberOfSplinePoints _spline_2) 2) 1))
    (bind _returnvalue (AddComponent|NODE_AddSplineMeshComponent-0))
    (Components|StaticMesh|SetStaticMesh _returnvalue (Variables|/Script/Engine.StaticMeshComponent|GetStaticMesh (Variables|SplineMeshActor|GetSplineMeshComponent)))
    (SplineMesh|SetStartandEnd _returnvalue (Spline|GetLocationandTangentatSplinePoint _spline_2 _index "Local") (Math|Vector|SelectVector (* (Math|Vector|Normalize (.Tangent (Spline|GetLocationandTangentatSplinePoint _spline_2 _index "Local"))) (* (Math|Vector|VectorLength (- (Spline|GetLocationandTangentatSplinePoint _spline_2 (+ _index 1) "Local") (Spline|GetLocationandTangentatSplinePoint _spline_2 _index "Local"))) (Variables|Default|GetTension))) (.Tangent (Spline|GetLocationandTangentatSplinePoint _spline_2 _index "Local")) (Variables|Default|GetAutoSmoothTangent)) (Spline|GetLocationandTangentatSplinePoint _spline_2 (+ _index 1) "Local") (Math|Vector|SelectVector (* (Math|Vector|Normalize (.Tangent (Spline|GetLocationandTangentatSplinePoint _spline_2 (+ _index 1) "Local"))) (* (Math|Vector|VectorLength (- (Spline|GetLocationandTangentatSplinePoint _spline_2 (+ _index 1) "Local") (Spline|GetLocationandTangentatSplinePoint _spline_2 _index "Local"))) (Variables|Default|GetTension))) (.Tangent (Spline|GetLocationandTangentatSplinePoint _spline_2 (+ _index 1) "Local")) (Variables|Default|GetAutoSmoothTangent)) true :!class /Script/Engine.SplineMeshComponent)
    (if (Variables|Default|GetUseScaling)
      (SplineMesh|SetStartScale _returnvalue (Math|Vector2D|MakeVector2D (.y (Spline|GetScaleatSplinePoint _spline_2 _index)) (.z (Spline|GetScaleatSplinePoint _spline_2 _index))) true :!class /Script/Engine.SplineMeshComponent)
      (SplineMesh|SetEndScale _returnvalue (Math|Vector2D|MakeVector2D (.y (Spline|GetScaleatSplinePoint _spline_2 (+ _index 1))) (.z (Spline|GetScaleatSplinePoint _spline_2 (+ _index 1)))) true :!class /Script/Engine.SplineMeshComponent))))
